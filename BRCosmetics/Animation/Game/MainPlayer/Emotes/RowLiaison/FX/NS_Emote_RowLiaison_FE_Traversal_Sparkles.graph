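# Niagara system for the RowLiaison traversal emote, adding sparkles, glows and a flickering light around the character and carpet.
# Glitter, glint, foot-glow and sparkle sprite emitters, some on GPU, place particles from the skeletal mesh, alongside a particle light.
# Most spawn rates switch on owner speed, with front-end overrides, and time-of-day functions adjust color and brightness.
# BlendSpace modules blend particle positions between world and local space, mostly by velocity and particle age.
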NiagaraSystem /BRCosmetics/Animation/Game/MainPlayer/Emotes/RowLiaison/FX/NS_Emote_RowLiaison_FE_Traversal_Sparkles
dependencies {
 usedDataInterfaces [
  &NiagaraDataInterfaceCurve
  &NiagaraDataInterfaceArrayFloat
  &NiagaraDataInterfaceColorCurve
  &NiagaraDataInterfaceSkeletalMesh
 ]
 usedDynamicInputs []
 usedModules [
  &self:SystemSpawnScript.SystemScriptSource.SystemScriptGraph.NiagaraNodeAssignment_0.SetVariables_#1
  &SystemState
  &BlendSpaceHelper
 ]
}
emitters {
 Body_Glitter {
  emitterSpawnScript [/SetVariables(Emitter.FE Particle Spawn Rate=25, Emitter.Idle Spawn Rate=25, Emitter.Max Traversal Speed=50, Emitter.Particle Spawn Rate=25) {isSetParameters=true, module=SetVariables_#2}]
  emitterUpdateScript [
   &EmitterState()
   &SpawnRate(
    SpawnRate=&FN_FrontEndFloatModifier(
     Gameplay=&MakeCustomFloatFromBool(
      Bool=&SetBoolByFloatComparison(A=&VectorLength(Vector=$User.OwnerLocalVelocity), B=140)
      True Float=$Emitter.Particle Spawn Rate
      False Float=$Emitter.Idle Spawn Rate
     )
     Front End=$Emitter.FE Particle Spawn Rate
    )
    Use Spawn Probability=false
   ) {module=SpawnRate001}
  ]
  particleSpawnScript [
   /SetVariables(Particles.VisibilityTag=&RandomRangeInt(Minimum=0, Maximum=2)) {isSetParameters=true, module=SetVariables_#3}
   &InitializeParticle(
    AdjustAlpha=false
    AdjustHue=true
    AdjustSaturation=false
    AdjustValue=false
    Color=<rgba 0.843 0.8 1 1>
    Color Mode=::NewEnumerator3
    Hue Shift Range:Vector2f=<-1 1>
    Lifetime Max=0.6
    Lifetime Min=0.3
    Lifetime Mode=::NewEnumerator1
    Sprite Size Mode=::NewEnumerator3
    Uniform Sprite Size=&Multiply_Float(
     A=&Multiply_Float(
      A=&SelectFloatFromArray(Direct Array Index=$Particles.VisibilityTag, Float Selection Array=DataInterface() {propertyValues {FloatData [100,50,15]}})
      B=&RandomRangeFloat(Minimum=0.75, Maximum=2)
     )
     B=1
    )
    UsePositionOffset=true
    Write Material Random=true
   )
   &CameraOffset(Camera Offset Amount=10)
   &SkeletalMeshLocation(
    MaskSpawnGroup=false
    Mesh Sampling Type=::NewEnumerator5
    Sampled Position Offset=<0 0 30>
    Skeletal Mesh=$System.Skeletal Mesh
    Socket Sampling Mode=::NewEnumerator5
   )
   &ShapeLocation(MaskSpawnGroup=false, Offset=<0 0 50>, Offset Mode=::NewEnumerator0, Override Seed=false, Sphere Radius=80, UseOwnerScale=true)
   &AddVelocity(Velocity=&RandomRangeVector(Minimum=<0 -20 0>, Maximum=<1 -60 1>))
  ]
  particleUpdateScript [
   &ParticleState()
   &ScaleSpriteSize(Scale Factor=&Multiply_Vector2D_ByFloat(
     Float=1
     Vector2D=&Vector2DFromFloat(Value=&FloatFromCurve(FloatCurve=DataInterface() {propertyValues {Curve {keys [
           {arriveTangent=9.0592, interpMode=RCIM_Cubic, leaveTangent=9.0592, tangentMode=RCTM_User, time=0, value=0}
           {interpMode=RCIM_Cubic, tangentMode=RCTM_User, time=0.2, value=1}
           {tangentMode=RCTM_User, time=1, value=0}
          ]}}}))
    ))
   /SetVariables(
    Particles.Color=&MultiplyLinearColor(
     Linear Color=$Particles.Color
     Scale Factor=&MakeVector4_FromLinearColor(LinearColor=<rgba 8.0099 9.8007 10 1>)
    )
    Particles.Color02:LinearColor=&MultiplyLinearColor(
     Linear Color=$Particles.Color
     Scale Factor=&MakeVector4_FromLinearColor(LinearColor=<rgba 0.9441 0.8079 1 1>)
    )
    Particles.Color03:LinearColor=&MultiplyLinearColor(
     Linear Color=$Particles.Color
     Scale Factor=&MakeVector4_FromLinearColor(LinearColor=<rgba 0.15 0.15 0.15 1>)
    )
    Particles.SpriteSize02:Vector2f=&Multiply_Vector2(A=$Particles.SpriteSize, B=&Vector2DFromFloat(Value=0.3))
   ) {isSetParameters=true, module=SetVariables_#4}
   &ScaleColor(
    Scale Alpha=&Multiply_Float(
     A=&FloatFromCurve(
      FloatCurve=DataInterface() {propertyValues {Curve {keys [
          {interpMode=RCIM_Cubic, tangentMode=RCTM_User, time=0, value=0}
          {arriveTangent=-2.4058, interpMode=RCIM_Cubic, leaveTangent=-2.4058, tangentMode=RCTM_User, time=0.26, value=0.8421}
          {arriveTangent=-1.1775, interpMode=RCIM_Cubic, leaveTangent=-1.1775, time=0.4613, value=0.0645}
          {arriveTangent=-0.1197, interpMode=RCIM_Cubic, leaveTangent=-0.1197, time=0.5898, value=0.4537}
          <at 1 0 RCIM_Cubic>
         ]}}}
      Scale Curve=&RandomRangeFloat(Minimum=0.1, Maximum=0.5)
     )
     B=&FloatFromCurve(
      FloatCurve=DataInterface() {propertyValues {Curve {keys [{tangentMode=RCTM_User, time=0, value=0}, {tangentMode=RCTM_User, time=1, value=1}]}}}
      CurveIndex=&Multiply_Float(
       A=&MakeFloatFromVector(Vector=Linked() {linkedVariable {name=Particles.Position, type=NiagaraPosition}}, Channel=::NewEnumerator2)
       B=$Engine.System.Age
      )
     )
    )
    Scale RGB=&MakeVectorFromLinearColorRGB(Color=&FNTimeOfDayColor(Night Color=<rgba 0.3 0.6267 1 10>, Morning Color=<rgba 1 1 1 1>, Day Color=<rgba 1 1 1 1>, Evening Color=<rgba 1 1 1 1>))
    ScaleA=true
    ScaleRGB=true
   )
   &CurlNoiseForce(Noise Strength=100, Pan Noise=false)
   &Drag(Drag=&RandomRangeFloat(Minimum=10, Maximum=12), Use Linear Drag=true, Use Rotational Drag=true)
   &SolveForcesAndVelocity(Clamp Velocity=false, Limit Acceleration=false)
   &BlendSpace("Position Blend (World <-> Local)":NiagaraFloat=&MakeCustomFloatFromBool(
     Bool=&SetBoolByFloatComparison(A=&VectorLength(Vector=$Engine.Owner.Velocity), B=400)
     True Float=1
     False Float=&MergeHelper_FloatSwitch(
      0=&FloatFromCurve(FloatCurve=DataInterface() {propertyValues {Curve {keys [<at 0 1 RCIM_Cubic>, {arriveTangent=-3.5186, interpMode=RCIM_Cubic, leaveTangent=-3.5186, tangentMode=RCTM_User, time=0.8, value=0}]}}})
      1=&FloatFromCurve(FloatCurve=DataInterface() {propertyValues {Curve {keys [<at 0 1 RCIM_Cubic>, {arriveTangent=-3.5186, interpMode=RCIM_Cubic, leaveTangent=-3.5186, tangentMode=RCTM_User, time=0.5, value=0}]}}})
      2=?
      3=?
      Index=&SpawnGroup()
      Num=::NewEnumerator0
     )
     Translation Contribution:NiagaraFloat=0.5
     Rotation Contribution:NiagaraFloat=0
    )) {module=BlendSpace002}
   &BlendSpace("Position Blend (World <-> Local)":NiagaraFloat=&FloatFromCurve(
     FloatCurve=DataInterface() {propertyValues {Curve {
        defaultValue=0
        keys [{interpMode=RCIM_Cubic, tangentMode=RCTM_User, time=0, value=1}, {interpMode=RCIM_Cubic, tangentMode=RCTM_User, time=0.5, value=0}]
       }}}
     Translation Contribution:NiagaraFloat=2
     Rotation Contribution:NiagaraFloat=0
    )) {enabled=false}
  ]
  rendererClasses [&NiagaraSpriteRendererProperties]
  renderers [
   {rendererClass=&NiagaraSpriteRendererProperties}
   {rendererClass=&NiagaraSpriteRendererProperties, rendererIndex=1}
   {rendererClass=&NiagaraSpriteRendererProperties, rendererIndex=2}
  ]
 }
 Carpet_Glint_Low-Med {
  emitterSpawnScript [/SetVariables(Emitter.FE Particle Spawn Rate=60, Emitter.Idle Spawn Rate=200, Emitter.Max Traversal Speed=150, Emitter.Particle Spawn Rate=60) {isSetParameters=true, module=SetVariables_#2}]
  emitterUpdateScript [
   &EmitterState()
   &SpawnBurst_Instantaneous(Spawn Group=2, Use Spawn Probability=false, UseLoopCountLimit=false)
   &SpawnRate(
    SpawnRate=&FN_FrontEndFloatModifier(
     Gameplay=&MakeCustomFloatFromBool(
      Bool=&SetBoolByFloatComparison(A=&VectorLength(Vector=$User.OwnerLocalVelocity), B=140)
      True Float=$Emitter.Particle Spawn Rate
      False Float=$Emitter.Idle Spawn Rate
     )
     Front End=$Emitter.FE Particle Spawn Rate
    )
    Use Spawn Probability=false
   ) {module=SpawnRate001}
  ]
  particleSpawnScript [
   /SetVariables(Particles.VisibilityTag=&RandomRangeInt(Minimum=0, Maximum=2)) {isSetParameters=true, module=SetVariables_#3}
   &InitializeParticle(
    AdjustAlpha=false
    AdjustHue=true
    AdjustSaturation=false
    AdjustValue=false
    Color=<rgba 0.843 0.8 1 1>
    Color Mode=::NewEnumerator3
    Hue Shift Range:Vector2f=<-1 1>
    Lifetime Max=0.85
    Lifetime Min=0.5
    Lifetime Mode=::NewEnumerator1
    Sprite Rotation Mode=::NewEnumerator2
    Sprite Size Mode=::NewEnumerator3
    Uniform Sprite Size=&Multiply_Float(
     A=&Multiply_Float(
      A=&SelectFloatFromArray(Direct Array Index=$Particles.VisibilityTag, Float Selection Array=DataInterface() {propertyValues {FloatData [100,30,15]}})
      B=&RandomRangeFloat(Minimum=0.6, Maximum=0.8)
     )
     B=1
    )
    UsePositionOffset=true
    Write Material Random=true
   )
   &CameraOffset(Camera Offset Amount=10)
   &SkeletalMeshLocation(Bone Sampling Mode=::NewEnumerator4, MaskSpawnGroup=false, Sampled Position Offset=<0 20 15>, Skeletal Mesh=$System.Skeletal Mesh)
   &ShapeLocation(
    Box Size=<95 145 5>
    MaskSpawnGroup=false
    Offset Mode=::NewEnumerator0
    Override Seed=false
    Shape Primitive=::NewEnumerator2
    Surface Only=false
    UseOwnerScale=true
   )
   &AddVelocity(Velocity=&RandomRangeVector(Minimum=<0 -100 0>, Maximum=<1 -350 1>))
  ]
  particleUpdateScript [
   &ParticleState()
   &ScaleColor(
    Scale Alpha=&Multiply_Float(
     A=&FloatFromCurve(
      FloatCurve=DataInterface() {propertyValues {Curve {keys [
          {interpMode=RCIM_Cubic, tangentMode=RCTM_User, time=0, value=0}
          {arriveTangent=-2.4058, interpMode=RCIM_Cubic, leaveTangent=-2.4058, tangentMode=RCTM_User, time=0.26, value=0.8421}
          {arriveTangent=-0.9001, interpMode=RCIM_Cubic, leaveTangent=-0.9001, time=0.4613, value=0.0645}
          {arriveTangent=1.9066, interpMode=RCIM_Cubic, leaveTangent=1.9066, time=0.5594, value=0.5726}
          {arriveTangent=-0.3157, interpMode=RCIM_Cubic, leaveTangent=-0.3157, tangentMode=RCTM_User, time=0.6774, value=0.4764}
          {arriveTangent=-7.1699, interpMode=RCIM_Cubic, leaveTangent=-7.1699, tangentMode=RCTM_User, time=1, value=0}
         ]}}}
      Scale Curve=&RandomRangeFloat(Minimum=0.1, Maximum=0.5)
     )
     B=&FloatFromCurve(
      FloatCurve=DataInterface() {propertyValues {Curve {
         defaultValue=0
         keys [{interpMode=RCIM_Cubic, tangentMode=RCTM_User, time=0, value=0}, {interpMode=RCIM_Cubic, tangentMode=RCTM_User, time=1, value=1}]
        }}}
      CurveIndex=&Multiply_Float(
       A=&MakeFloatFromVector(Vector=Linked() {linkedVariable {name=Particles.Position, type=NiagaraPosition}}, Channel=::NewEnumerator2)
       B=$Engine.System.Age
      )
     )
    )
    Scale RGB=&MakeVectorFromLinearColorRGB(Color=&FNTimeOfDayColor(Night Color=<rgba 0.3 0.6267 1 5>, Morning Color=<rgba 1 1 1 1>, Day Color=<rgba 5 5 5 1>, Evening Color=<rgba 1 1 1 1>))
    ScaleA=true
    ScaleRGB=true
   )
   &ScaleSpriteSize(Scale Factor=&Multiply_Vector2D_ByFloat(
     Float=1
     Vector2D=&Vector2DFromFloat(Value=&FloatFromCurve(
       FloatCurve=DataInterface() {propertyValues {Curve {keys [
           {arriveTangent=9.0592, interpMode=RCIM_Cubic, leaveTangent=9.0592, tangentMode=RCTM_User, time=0, value=0}
           {interpMode=RCIM_Cubic, tangentMode=RCTM_User, time=0.2, value=1}
           {tangentMode=RCTM_User, time=1, value=0}
          ]}}}
       Scale Curve=&RandomRangeFloat(Minimum=0.3, Maximum=1)
      ))
    ))
   /SetVariables(
    Particles.Color=&MultiplyLinearColor(
     Linear Color=$Particles.Color
     Scale Factor=&MakeVector4_FromLinearColor(LinearColor=<rgba 8.0099 9.8007 10 1>)
    )
    Particles.Color02:LinearColor=&MultiplyLinearColor(
     Linear Color=$Particles.Color
     Scale Factor=&MakeVector4_FromLinearColor(LinearColor=<rgba 0.9441 0.8079 1 1>)
    )
    Particles.Color03:LinearColor=&MultiplyLinearColor(
     Linear Color=$Particles.Color
     Scale Factor=&MakeVector4_FromLinearColor(LinearColor=<rgba 0.15 0.15 0.15 1>)
    )
    Particles.SpriteSize02:Vector2f=&Multiply_Vector2(A=$Particles.SpriteSize, B=&Vector2DFromFloat(Value=0.3))
   ) {isSetParameters=true, module=SetVariables_#4}
   &CurlNoiseForce(Noise Strength=100, Pan Noise=false)
   &Drag(
    Drag=&FloatFromCurve(
     FloatCurve=DataInterface() {propertyValues {Curve {
        defaultValue=0
        keys [{interpMode=RCIM_Cubic, tangentMode=RCTM_User, time=0, value=1}, {interpMode=RCIM_Cubic, tangentMode=RCTM_User, time=1, value=0}]
       }}}
     Scale Curve=&RandomRangeFloat(Minimum=10, Maximum=12)
    )
    Use Linear Drag=true
    Use Rotational Drag=true
   ) {module=Drag001}
   &SolveForcesAndVelocity(Clamp Velocity=false, Limit Acceleration=false)
   &BlendSpace("Position Blend (World <-> Local)":NiagaraFloat=&MakeCustomFloatFromBool(
     Bool=&SetBoolByFloatComparison(A=&VectorLength(Vector=$Engine.Owner.Velocity), B=400)
     True Float=1
     False Float=&FloatFromCurve(FloatCurve=DataInterface() {propertyValues {Curve {keys [<at 0 1 RCIM_Cubic>, {arriveTangent=-12.1775, interpMode=RCIM_Cubic, leaveTangent=-12.1775, tangentMode=RCTM_User, time=0.25, value=0}]}}})
     Translation Contribution:NiagaraFloat=1
     Rotation Contribution:NiagaraFloat=0
    )) {module=BlendSpace001}
  ]
  rendererClasses [&NiagaraSpriteRendererProperties]
  renderers [
   {rendererClass=&NiagaraSpriteRendererProperties}
   {rendererClass=&NiagaraSpriteRendererProperties, rendererIndex=1}
  ]
 }
 Glow_on_feet_Low-_Med {
  emitterSpawnScript [/SetVariables(Emitter.FE Particle Spawn Rate=85, Emitter.Idle Spawn Rate=135, Emitter.Max Traversal Speed=140, Emitter.Particle Spawn Rate=130) {isSetParameters=true, module=SetVariables_#5}]
  emitterUpdateScript [
   &EmitterState()
   &SpawnRate(
    SpawnRate=&FN_FrontEndFloatModifier(
     Gameplay=&MakeCustomFloatFromBool(
      Bool=&SetBoolByFloatComparison(A=&VectorLength(Vector=$User.OwnerLocalVelocity), B=139)
      True Float=$Emitter.Particle Spawn Rate
      False Float=$Emitter.Idle Spawn Rate
     )
     Front End=$Emitter.FE Particle Spawn Rate
    )
    Use Spawn Probability=false
   )
  ]
  particleSpawnScript [
   &InitializeParticle(
    Lifetime=&RandomRangeFloat(Minimum=0.75, Maximum=1)
    Sprite Rotation Mode=::NewEnumerator2
    Sprite Size Mode=::NewEnumerator3
    Sprite UV Mode=::NewEnumerator3
    Uniform Sprite Size=&Multiply_Float(
     A=&RandomRangeFloat(Minimum=65, Maximum=120)
     B=&self:SpawnGrp_0_1_Mask(Masked Float:NiagaraFloat=0.8, Passthrough Float:NiagaraFloat=1)
    )
    UsePositionOffset=false
    Write Material Random=true
   )
   &CameraOffset(Camera Offset Amount=&RandomRangeFloat(Minimum=0, Maximum=30))
   &AddVelocity(Velocity=&MakeVector(X=0, Y=&RandomRangeFloat(Minimum=-10, Maximum=-35), Z=0)) {enabled=false}
   /Game/Sprite/FN_SpriteAlignmentAndFacing(SpriteAlignment:Vector3f=<0 1 0>, SpriteFacing:Vector3f=<0 0 1>)
  ]
  particleUpdateScript [
   &ParticleState()
   &SkeletalMeshLocation(
    MaskSpawnGroup=false
    Mesh Sampling Type=::NewEnumerator5
    Orientation Sampling=::NewEnumerator1
    Position Sampling=::NewEnumerator1
    Sampled Position Offset=&RandomRangeVector(Minimum=<-20 50 10>, Maximum=<20 -50 15>)
    Sampled Position Offset Coordinate Space:ENiagaraSkelMeshTransforms=::NewEnumerator2
    Skeletal Mesh=$System.Skeletal Mesh
    Socket Sampling Mode=::NewEnumerator5
   ) {module=SkeletalMeshLocation001}
   &ScaleColor(
    Scale Alpha=&FloatFromCurve(
     FloatCurve=DataInterface() {propertyValues {Curve {keys [<at 0 0 RCIM_Cubic>, {interpMode=RCIM_Cubic, tangentMode=RCTM_Break, time=0.2, value=1}, <at 1 0>]}}}
     Scale Curve=&Multiply_Float(
      A=&FN_FrontEndFloatModifier(Gameplay=1, Front End=0.3)
      B=&self:SpawnGrp_0_1_Mask(
       Masked Float:NiagaraFloat=&RandomRangeFloat(Minimum=0.2, Maximum=0.4)
       Passthrough Float:NiagaraFloat=1
      )
     )
    )
    Scale RGB=&MakeVectorFromLinearColorRGB(Color=&ColorFromCurve(ColorCurve=DataInterface() {propertyValues {AlphaCurve {keys []}, BlueCurve {keys [<at 0.436 3>]}, GreenCurve {keys [<at 0.436 2.4023>]}, RedCurve {keys [<at 0.436 1.9531>]}}}))
    ScaleA=true
    ScaleRGB=true
   ) {module=ScaleColor001}
   &ScaleColor(
    Scale Alpha=&FloatFromCurve(
     FloatCurve=DataInterface() {propertyValues {Curve {keys [
         {interpMode=RCIM_Cubic, leaveTangentWeight=0.0833, tangentMode=RCTM_User, time=0, value=0}
         {
          arriveTangent=0.7559
          arriveTangentWeight=0.1572
          interpMode=RCIM_Cubic
          leaveTangent=0.7559
          leaveTangentWeight=0.1572
          tangentMode=RCTM_User
          time=0.2014
          value=0.483
         }
         {arriveTangentWeight=0.1681, interpMode=RCIM_Cubic, leaveTangentWeight=0.0841, tangentMode=RCTM_User, time=0.5, value=1}
         {
          arriveTangent=-0.6245
          arriveTangentWeight=0.1572
          interpMode=RCIM_Cubic
          leaveTangent=-0.6245
          leaveTangentWeight=0.1572
          tangentMode=RCTM_User
          time=0.7889
          value=0.5141
         }
         {arriveTangentWeight=0.0833, interpMode=RCIM_Cubic, tangentMode=RCTM_User, time=1, value=0.2}
        ]}}}
     Scale Curve=&FNTimeOfDayScalar(Night Scale=0.2, Morning Scale=0.1, Day Scale=0.3, Evening Scale=0.1)
    )
    ScaleA=true
    ScaleRGB=false
   ) {module=ScaleColor002}
   &ScaleSpriteSize(Uniform Curve Sprite Scale=DataInterface() {propertyValues {Curve {keys [
        {arriveTangent=3.5003, interpMode=RCIM_Cubic, leaveTangent=3.5003, tangentMode=RCTM_User, time=0, value=0.5}
        {arriveTangent=0.5833, interpMode=RCIM_Cubic, leaveTangent=0.5833, time=0.18, value=0.9}
        <at 1.2 1.2 RCIM_Cubic>
       ]}}})
   &DynamicMaterialParameters(
    Index 0 Param 1=&FloatFromCurve(FloatCurve=DataInterface() {propertyValues {Curve {keys [<at 0 0 RCIM_Cubic>,<at 1.0011 0.2>]}}})
    Index 1 Param 1=&FloatFromCurve(
     FloatCurve=DataInterface() {propertyValues {Curve {keys [<at 0 0>,<at 0.25 1>]}}}
     Scale Curve=&RandomRangeFloat(Minimum=2, Maximum=1)
    )
    Index 1 Param 2=&FloatFromCurve(
     FloatCurve=DataInterface() {propertyValues {Curve {keys [
         {interpMode=RCIM_Cubic, leaveTangent=3.7422, tangentMode=RCTM_Break, time=0, value=0.2}
         {interpMode=RCIM_Cubic, leaveTangent=0.9066, time=1, value=2}
        ]}}}
     Scale Curve=&RandomRangeFloat(Minimum=0.5, Maximum=1)
    )
    Index 1 Param 3=&FloatFromCurve(FloatCurve=DataInterface() {propertyValues {Curve {keys [
         {interpMode=RCIM_Cubic, leaveTangent=2.8445, tangentMode=RCTM_Break, time=0, value=0.2}
         {interpMode=RCIM_Cubic, leaveTangent=63.2978, time=0.25, value=1}
        ]}}})
    Index 1 Param 4=&FloatFromCurve(
     FloatCurve=DataInterface() {propertyValues {Curve {keys [<at 0 1>,<at 1 0.8>]}}}
     Scale Curve=&RandomRangeFloat(Minimum=0.5, Maximum=1)
    )
    Index 2 Param 1=&FloatFromCurve(FloatCurve=DataInterface() {propertyValues {Curve {keys [{interpMode=RCIM_Cubic, leaveTangent=3.2309, tangentMode=RCTM_Break, time=0.1, value=0}, <at 1 1>]}}})
    Index 2 Param 2=$Particles.MaterialRandom
    Index 2 Param 3=&FloatFromCurve(FloatCurve=DataInterface() {propertyValues {Curve {keys [
         {interpMode=RCIM_Cubic, tangentMode=RCTM_Break, time=0, value=1}
         {interpMode=RCIM_Cubic, leaveTangent=11.1233, tangentMode=RCTM_Break, time=1, value=2}
        ]}}})
    Index 2 Param 4=0.75
    Module.Index0Param0Write=true
    Module.Index0Param1Write=true
    Module.Index0Param2Write=true
    Module.Index0Param3Write=true
    Module.Index1Param0Write=true
    Module.Index1Param1Write=true
    Module.Index1Param2Write=true
    Module.Index1Param3Write=true
    Module.Index2Param0Write=true
    Module.Index2Param1Write=true
    Module.Index2Param2Write=true
    Module.Index2Param3Write=true
    Write Parameter Index 1=true
    Write Parameter Index 2=true
   ) {module=DynamicMaterialParameters001}
   &BlendSpace(
    "Position Blend (World <-> Local)":NiagaraFloat=&MakeCustomFloatFromBool(
     Bool=&SetBoolByFloatComparison(A=&VectorLength(Vector=$Engine.Owner.Velocity), B=139)
     True Float=1
     False Float=&FloatFromCurve(FloatCurve=DataInterface() {propertyValues {Curve {
         defaultValue=0
         keys [{interpMode=RCIM_Cubic, tangentMode=RCTM_User, time=0, value=1}, {interpMode=RCIM_Cubic, tangentMode=RCTM_User, time=0.15, value=0}]
        }}})
     Translation Contribution:NiagaraFloat=0
     Rotation Contribution:NiagaraFloat=0
    )
    Position Coordinate Space=2
   ) {module=BlendSpace002}
   &Drag(Drag=&RandomRangeFloat(Minimum=2, Maximum=5), Use Linear Drag=true, Use Rotational Drag=true)
   &SolveForcesAndVelocity(Clamp Velocity=false, Limit Acceleration=false)
   &OffsetPosition(
    Position Offset=&MakeVector(
     X=0
     Y=&FloatFromCurve(
      FloatCurve=DataInterface() {propertyValues {Curve {
         defaultValue=1
         keys [{interpMode=RCIM_Cubic, tangentMode=RCTM_User, time=0, value=0}, {interpMode=RCIM_Cubic, tangentMode=RCTM_User, time=1, value=1}]
        }}}
      Scale Curve=&RandomRangeFloat(Minimum=-5, Maximum=-10)
     )
     Z=0
    )
    Position Offset Coordinate Space:ENiagaraCoordinateSpace=::Local
   )
  ]
  rendererClasses [&NiagaraSpriteRendererProperties]
  renderers [{rendererClass=&NiagaraSpriteRendererProperties}]
 }
 Glow_on_feet__filler_for_ {
  simTarget=GPUComputeSim
  emitterSpawnScript [/SetVariables(Emitter.FE Particle Spawn Rate=85, Emitter.Idle Spawn Rate=135, Emitter.Max Traversal Speed=140, Emitter.Particle Spawn Rate=130) {isSetParameters=true, module=SetVariables_#5}]
  emitterUpdateScript [
   &EmitterState()
   &SpawnRate(
    SpawnRate=&FN_FrontEndFloatModifier(
     Gameplay=&MakeCustomFloatFromBool(
      Bool=&SetBoolByFloatComparison(A=&VectorLength(Vector=$User.OwnerLocalVelocity), B=139)
      True Float=$Emitter.Particle Spawn Rate
      False Float=$Emitter.Idle Spawn Rate
     )
     Front End=$Emitter.FE Particle Spawn Rate
    )
    Use Spawn Probability=false
   )
  ]
  particleSpawnScript [
   &InitializeParticle(
    Lifetime=&RandomRangeFloat(Minimum=0.75, Maximum=1)
    Sprite Rotation Mode=::NewEnumerator2
    Sprite Size Mode=::NewEnumerator3
    Sprite UV Mode=::NewEnumerator3
    Uniform Sprite Size=&Multiply_Float(
     A=&RandomRangeFloat(Minimum=75, Maximum=135)
     B=&self:SpawnGrp_0_1_Mask(Masked Float:NiagaraFloat=1, Passthrough Float:NiagaraFloat=1.2)
    )
    UsePositionOffset=false
    Write Material Random=true
   )
   &CameraOffset(Camera Offset Amount=&RandomRangeFloat(Minimum=0, Maximum=30))
   &FN_SpriteAlignmentAndFacing(SpriteAlignment:Vector3f=<0 1 0>, SpriteFacing:Vector3f=<0 0 1>)
  ]
  particleUpdateScript [
   &ParticleState()
   &SkeletalMeshLocation(
    MaskSpawnGroup=false
    Mesh Sampling Type=::NewEnumerator5
    Orientation Sampling=::NewEnumerator1
    Position Sampling=::NewEnumerator1
    Sampled Position Offset=&RandomRangeVector(Minimum=<-35 65 10>, Maximum=<35 -75 15>)
    Sampled Position Offset Coordinate Space:ENiagaraSkelMeshTransforms=::NewEnumerator2
    Skeletal Mesh=$System.Skeletal Mesh
    Socket Sampling Mode=::NewEnumerator5
   ) {module=SkeletalMeshLocation001}
   &ScaleColor(
    Scale Alpha=&FloatFromCurve(
     FloatCurve=DataInterface() {propertyValues {Curve {keys [<at 0 0 RCIM_Cubic>, {interpMode=RCIM_Cubic, tangentMode=RCTM_Break, time=0.2, value=1}, <at 1 0>]}}}
     Scale Curve=&Multiply_Float(
      A=&FN_FrontEndFloatModifier(Gameplay=1, Front End=0.3)
      B=&self:SpawnGrp_0_1_Mask(
       Masked Float:NiagaraFloat=&RandomRangeFloat(Minimum=0.2, Maximum=0.4)
       Passthrough Float:NiagaraFloat=1
      )
     )
    )
    Scale RGB=&MakeVectorFromLinearColorRGB(Color=&ColorFromCurve(ColorCurve=DataInterface() {propertyValues {AlphaCurve {keys []}, BlueCurve {keys [<at 0.436 3>]}, GreenCurve {keys [<at 0.436 2.4023>]}, RedCurve {keys [<at 0.436 1.9531>]}}}))
    ScaleA=true
    ScaleRGB=true
   ) {module=ScaleColor001}
   &ScaleColor(
    Scale Alpha=&FloatFromCurve(
     FloatCurve=DataInterface() {propertyValues {Curve {keys [
         {interpMode=RCIM_Cubic, leaveTangentWeight=0.0833, tangentMode=RCTM_User, time=0, value=0}
         {
          arriveTangent=0.7559
          arriveTangentWeight=0.1572
          interpMode=RCIM_Cubic
          leaveTangent=0.7559
          leaveTangentWeight=0.1572
          tangentMode=RCTM_User
          time=0.2014
          value=0.483
         }
         {arriveTangentWeight=0.1681, interpMode=RCIM_Cubic, leaveTangentWeight=0.0841, tangentMode=RCTM_User, time=0.5, value=1}
         {
          arriveTangent=-0.6245
          arriveTangentWeight=0.1572
          interpMode=RCIM_Cubic
          leaveTangent=-0.6245
          leaveTangentWeight=0.1572
          tangentMode=RCTM_User
          time=0.7889
          value=0.5141
         }
         {arriveTangentWeight=0.0833, interpMode=RCIM_Cubic, tangentMode=RCTM_User, time=1, value=0.2}
        ]}}}
     Scale Curve=&FNTimeOfDayScalar(Night Scale=0.2, Morning Scale=0.1, Day Scale=0.3, Evening Scale=0.1)
    )
    ScaleA=true
    ScaleRGB=false
   ) {module=ScaleColor002}
   &ScaleSpriteSize(Uniform Curve Sprite Scale=DataInterface() {propertyValues {Curve {keys [
        {arriveTangent=3.5003, interpMode=RCIM_Cubic, leaveTangent=3.5003, tangentMode=RCTM_User, time=0, value=0.5}
        {arriveTangent=0.5833, interpMode=RCIM_Cubic, leaveTangent=0.5833, time=0.18, value=0.9}
        <at 1.2 1.2 RCIM_Cubic>
       ]}}})
   &DynamicMaterialParameters(
    Index 0 Param 1=&FloatFromCurve(FloatCurve=DataInterface() {propertyValues {Curve {keys [<at 0 0 RCIM_Cubic>,<at 1.0011 0.2>]}}})
    Index 1 Param 1=&FloatFromCurve(
     FloatCurve=DataInterface() {propertyValues {Curve {keys [<at 0 0>,<at 0.25 1>]}}}
     Scale Curve=&RandomRangeFloat(Minimum=2, Maximum=1)
    )
    Index 1 Param 2=&FloatFromCurve(
     FloatCurve=DataInterface() {propertyValues {Curve {keys [
         {interpMode=RCIM_Cubic, leaveTangent=3.7422, tangentMode=RCTM_Break, time=0, value=0.2}
         {interpMode=RCIM_Cubic, leaveTangent=0.9066, time=1, value=2}
        ]}}}
     Scale Curve=&RandomRangeFloat(Minimum=0.5, Maximum=1)
    )
    Index 1 Param 3=&FloatFromCurve(FloatCurve=DataInterface() {propertyValues {Curve {keys [
         {interpMode=RCIM_Cubic, leaveTangent=2.8445, tangentMode=RCTM_Break, time=0, value=0.2}
         {interpMode=RCIM_Cubic, leaveTangent=63.2978, time=0.25, value=1}
        ]}}})
    Index 1 Param 4=&FloatFromCurve(
     FloatCurve=DataInterface() {propertyValues {Curve {keys [<at 0 1>,<at 1 0.8>]}}}
     Scale Curve=&RandomRangeFloat(Minimum=0.5, Maximum=1)
    )
    Index 2 Param 1=&FloatFromCurve(FloatCurve=DataInterface() {propertyValues {Curve {keys [{interpMode=RCIM_Cubic, leaveTangent=3.2309, tangentMode=RCTM_Break, time=0.1, value=0}, <at 1 1>]}}})
    Index 2 Param 2=$Particles.MaterialRandom
    Index 2 Param 3=&FloatFromCurve(FloatCurve=DataInterface() {propertyValues {Curve {keys [
         {interpMode=RCIM_Cubic, tangentMode=RCTM_Break, time=0, value=1}
         {interpMode=RCIM_Cubic, leaveTangent=11.1233, tangentMode=RCTM_Break, time=1, value=2}
        ]}}})
    Index 2 Param 4=0.75
    Module.Index0Param0Write=true
    Module.Index0Param1Write=true
    Module.Index0Param2Write=true
    Module.Index0Param3Write=true
    Module.Index1Param0Write=true
    Module.Index1Param1Write=true
    Module.Index1Param2Write=true
    Module.Index1Param3Write=true
    Module.Index2Param0Write=true
    Module.Index2Param1Write=true
    Module.Index2Param2Write=true
    Module.Index2Param3Write=true
    Write Parameter Index 1=true
    Write Parameter Index 2=true
   ) {module=DynamicMaterialParameters001}
   &BlendSpace(
    "Position Blend (World <-> Local)":NiagaraFloat=&MakeCustomFloatFromBool(
     Bool=&SetBoolByFloatComparison(A=&VectorLength(Vector=$Engine.Owner.Velocity), B=139)
     True Float=1
     False Float=&FloatFromCurve(FloatCurve=DataInterface() {propertyValues {Curve {
         defaultValue=0
         keys [{interpMode=RCIM_Cubic, tangentMode=RCTM_User, time=0, value=1}, {interpMode=RCIM_Cubic, tangentMode=RCTM_User, time=0.15, value=0}]
        }}})
     Translation Contribution:NiagaraFloat=0
     Rotation Contribution:NiagaraFloat=0
    )
    Position Coordinate Space=2
   ) {module=BlendSpace002}
   &Drag(Drag=&RandomRangeFloat(Minimum=2, Maximum=5), Use Linear Drag=true, Use Rotational Drag=true)
   &SolveForcesAndVelocity(Clamp Velocity=false, Limit Acceleration=false)
   &OffsetPosition(
    Position Offset=&MakeVector(
     X=0
     Y=&FloatFromCurve(
      FloatCurve=DataInterface() {propertyValues {Curve {
         defaultValue=1
         keys [{interpMode=RCIM_Cubic, tangentMode=RCTM_User, time=0, value=0}, {interpMode=RCIM_Cubic, tangentMode=RCTM_User, time=1, value=1}]
        }}}
      Scale Curve=&RandomRangeFloat(Minimum=-10, Maximum=-20)
     )
     Z=0
    )
    Position Offset Coordinate Space:ENiagaraCoordinateSpace=::Local
   )
  ]
  rendererClasses [&NiagaraSpriteRendererProperties]
  renderers [{rendererClass=&NiagaraSpriteRendererProperties}]
 }
 Light {
  emitterSpawnScript [/SetVariables(Emitter.FE Particle Spawn Rate=60, Emitter.Max Traversal Speed=50, Emitter.Particle Spawn Rate=60) {isSetParameters=true}]
  emitterUpdateScript [
   &EmitterState()
   &SpawnBurst_Instantaneous(Spawn Group=5, Use Spawn Probability=false, UseLoopCountLimit=false) {module=SpawnBurst_Instantaneous001}
   /SetVariables(Emitter.Flicker:NiagaraFloat=&Add_Float(A=&Multiply_Float(A=$Emitter.Waveform.Waveform, B=0.1), B=0.5)) {isSetParameters=true}
  ]
  particleSpawnScript [
   /SetVariables(Particles.VisibilityTag=&SpawnGroup()) {isSetParameters=true}
   &PartitionParticles()
   &InitializeParticle(
    Lifetime=5
    Mesh Scale=&Multiply_VectorByFloat(Vector=<3.5 0.5 1.5>, Float=0.07)
    Mesh Scale Mode=::NewEnumerator4
    Position Mode=::NewEnumerator1
    Sprite Rotation Angle=&MergeHelper_FloatSwitch(
     0=&UniformRangedFloat(Minimum=0, Maximum=180)
     1=&RandomRangeFloat(Minimum=-180, Maximum=200)
     2=&RandomRangeFloat(Minimum=-360, Maximum=360)
     3=100
     4=0
     Index=&SpawnGroup()
     Num=::NewEnumerator3
    )
    Sprite Rotation Mode=::NewEnumerator0
    Sprite Size=&MergeHelper_Vector2Switch(
     0=&Vector2DFromFloat(Value=&Multiply_Float(A=0.4, B=&UniformRangedFloat(Minimum=35, Maximum=50)))
     1=&Vector2DFromFloat(Value=&RandomRangeFloat(Minimum=5, Maximum=10))
     2=&RandomRangeVector2D(Minimum=<85 85>, Maximum=<175 175>)
     3=<5 5>
     4=&RandomRangeVector2D(Minimum=<20 20>, Maximum=<35 35>)
     Index=&SpawnGroup()
     Num=::NewEnumerator3
    )
    Sprite Size Mode=::NewEnumerator4
    UsePositionOffset=false
    Write Material Random=true
   ) {module=InitializeParticle001}
  ]
  particleUpdateScript [
   &ParticleState()
   &SkeletalMeshLocation(
    Bone / Socket Index=$Particles.PartitionParticles.Partition
    MaskSpawnGroup=false
    Mesh Sampling Type=::NewEnumerator5
    Orientation Sampling=::NewEnumerator1
    Position Sampling=::NewEnumerator1
    Sampled Position Offset=<0 35 15>
    Sampled Position Offset Coordinate Space:ENiagaraSkelMeshTransforms=::NewEnumerator2
    Skeletal Mesh=$System.Skeletal Mesh
    Socket Sampling Mode=::NewEnumerator5
   )
   &SolveForcesAndVelocity(Clamp Velocity=false, Limit Acceleration=false)
   &FN_Light_Attributes(
    Color=<rgba 3.63 4.1486 5 1>
    Create Color Attribute=true
    Create Exponent Attribute=true
    Create Position Attribute=true
    Create Radius Attribute=true
    Enable Particle Light Enable=true
    Enable Visibility Tag=true
    Exponent:NiagaraFloat=&FN_FrontEndFloatModifier(Gameplay=7, Front End=8)
    Light Enabled=true
    Position=<0 0 35>
    Radius=&FN_FrontEndFloatModifier(Gameplay=50, Front End=20)
    Scale Color=&Multiply_Float(
     A=0.5
     B=&Multiply_Float(
      A=&Waveform("Waveform [1]"=::NewEnumerator0, "[1] Frequency"=3, "[1] Amplitude Min/Max"=<0.7 1>, "[1] Scale"=2.6368)
      B=&FNTimeOfDayScalar(Night Scale=0.5, Morning Scale=0.4, Day Scale=0.45, Evening Scale=0.6)
     )
    )
    VisibilityTag=5
   )
  ]
  rendererClasses [&NiagaraLightRendererProperties]
  renderers [{rendererClass=&NiagaraLightRendererProperties, rendererIndex=0}]
 }
 Med_sparkles {
  simTarget=GPUComputeSim
  emitterUpdateScript [
   &EmitterState()
   &SpawnBurst_Instantaneous(Spawn Group=2, Use Spawn Probability=false, UseLoopCountLimit=false) {module=SpawnBurst_Instantaneous003}
   &SpawnRate(SpawnRate=35, Use Spawn Probability=false)
  ]
  particleSpawnScript [
   /SetVariables(Particles.VisibilityTag=&SpawnGroup()) {isSetParameters=true}
   &InitializeParticle(
    Color Mode=::NewEnumerator0
    Lifetime=&MergeHelper_FloatSwitch(
     0=&RandomRangeFloat(Minimum=0.3, Maximum=0.6)
     1=5
     2=?
     3=?
     Index=&SpawnGroup()
     Num=::NewEnumerator0
    )
    Mesh Scale Mode=::NewEnumerator3
    Mesh Uniform Scale=5
    Position Offset=&MergeHelper_VectorSwitch(0=<0 0 0>, 1=<0 0 0>, 2=<0 0 0>, 3=?, Index=&SpawnGroup(), Num=::NewEnumerator1)
    Sprite Rotation Mode=::NewEnumerator2
    Sprite Size Mode=::NewEnumerator3
    Uniform Sprite Size=&Multiply_Float(
     A=&Multiply_Float(
      A=&SelectFloatFromArray(Direct Array Index=$Particles.VisibilityTag, Float Selection Array=DataInterface() {propertyValues {FloatData [100,30,15]}})
      B=&RandomRangeFloat(Minimum=0.3, Maximum=0.5)
     )
     B=0.5
    )
    UsePositionOffset=false
    Write Material Random=true
   )
   /SetVariables(
    Particles.1 dynparam:Vector4f=&MakeVector4(X=0, Y=&RandomRangeFloat(Minimum=0, Maximum=4), Z=0, W=2.5)
    Particles.SpriteAlignment=&TransformVector(Vector=<0 0 0>, Destination Space=::World)
   ) {isSetParameters=true}
   &CameraOffset(Camera Offset Amount=&MergeHelper_FloatSwitch(0=0, 1=3, 2=?, 3=?, Index=&SpawnGroup(), Num=::NewEnumerator0))
   &SkeletalMeshLocation(
    MaskSpawnGroup=false
    Mesh Sampling Type=::NewEnumerator5
    Orientation Sampling=::NewEnumerator1
    Position Sampling=::NewEnumerator1
    Sampled Position Offset=&RandomRangeVector(Minimum=<-65 80 2>, Maximum=<65 -70 7>)
    Sampled Position Offset Coordinate Space:ENiagaraSkelMeshTransforms=::NewEnumerator2
    Skeletal Mesh=$System.Skeletal Mesh
    Socket Sampling Mode=::NewEnumerator5
   ) {module=SkeletalMeshLocation001}
  ]
  particleUpdateScript [
   &ParticleState()
   &Drag(
    Drag=&MergeHelper_FloatSwitch(
     0=0
     1=0
     2=0
     3=0
     4=&FloatFromCurve(
      FloatCurve=DataInterface() {propertyValues {Curve {defaultValue=1, keys [{tangentMode=RCTM_User, time=0, value=0}, {tangentMode=RCTM_User, time=1, value=1}]}}}
      Scale Curve=60
     )
     Index=&SpawnGroup()
     Num=::NewEnumerator3
    )
    Use Linear Drag=true
    Use Rotational Drag=true
   )
   &SolveForcesAndVelocity(Clamp Velocity=false, Limit Acceleration=false)
   &ScaleColor(
    Scale Alpha=&Multiply_Float(
     A=&FloatFromCurve(
      FloatCurve=DataInterface() {propertyValues {Curve {keys [
          {interpMode=RCIM_Cubic, tangentMode=RCTM_User, time=0, value=0}
          {arriveTangent=-2.4058, interpMode=RCIM_Cubic, leaveTangent=-2.4058, tangentMode=RCTM_User, time=0.26, value=0.8421}
          {arriveTangent=-0.8171, interpMode=RCIM_Cubic, leaveTangent=-0.8171, time=0.4613, value=0.0645}
          {arriveTangent=1.2485, interpMode=RCIM_Cubic, leaveTangent=1.2485, time=0.5898, value=0.5726}
          {arriveTangent=-1.3959, interpMode=RCIM_Cubic, leaveTangent=-1.3959, time=0.8782, value=0.5849}
          <at 1 0 RCIM_Cubic>
         ]}}}
      Scale Curve=&RandomRangeFloat(Minimum=0.1, Maximum=0.5)
     )
     B=&FloatFromCurve(
      FloatCurve=DataInterface() {propertyValues {Curve {
         defaultValue=0
         keys [{interpMode=RCIM_Cubic, tangentMode=RCTM_User, time=0, value=0}, {interpMode=RCIM_Cubic, tangentMode=RCTM_User, time=1, value=1}]
        }}}
      CurveIndex=&Multiply_Float(
       A=&MakeFloatFromVector(Vector=Linked() {linkedVariable {name=Particles.Position, type=NiagaraPosition}}, Channel=::NewEnumerator2)
       B=$Engine.System.Age
      )
      Scale Curve=0.7
     )
    )
    Scale RGB=&MakeVectorFromLinearColorRGB(Color=&FNTimeOfDayColor(Night Color=<rgba 1.6688 1.8233 2 5>, Morning Color=<rgba 1 1 1 1>, Day Color=<rgba 5 5 5 0.7>, Evening Color=<rgba 1 1 1 1>))
    ScaleA=true
    ScaleRGB=true
   )
   &ScaleMeshSize()
   &ScaleSpriteSize(Scale Factor=&Multiply_Vector2D_ByFloat(
     Float=1
     Vector2D=&Vector2DFromFloat(Value=&FloatFromCurve(
       FloatCurve=DataInterface() {propertyValues {Curve {keys [
           {arriveTangent=9.0592, interpMode=RCIM_Cubic, leaveTangent=9.0592, tangentMode=RCTM_User, time=0, value=0}
           {interpMode=RCIM_Cubic, tangentMode=RCTM_User, time=0.2, value=1}
           {tangentMode=RCTM_User, time=1, value=0}
          ]}}}
       Scale Curve=&FN_FrontEndFloatModifier(
        Gameplay=&RandomRangeFloat(Minimum=0.2, Maximum=0.85)
        Front End=&RandomRangeFloat(Minimum=0.5, Maximum=0.7)
       )
      ))
    ))
   /SetVariables(
    Particles.Color=&MultiplyLinearColor(
     Linear Color=$Particles.Color
     Scale Factor=&MakeVector4_FromLinearColor(LinearColor=<rgba 8.0099 9.8007 10 1>)
    )
    Particles.Color02:LinearColor=&MultiplyLinearColor(
     Linear Color=$Particles.Color
     Scale Factor=&MakeVector4_FromLinearColor(LinearColor=<rgba 0.9441 0.8079 1 1>)
    )
    Particles.Color03:LinearColor=&MultiplyLinearColor(
     Linear Color=$Particles.Color
     Scale Factor=&MakeVector4_FromLinearColor(LinearColor=<rgba 0.15 0.15 0.15 1>)
    )
    Particles.SpriteSize02:Vector2f=&Multiply_Vector2(A=$Particles.SpriteSize, B=&Vector2DFromFloat(Value=0.3))
   ) {isSetParameters=true}
   &BlendSpace(Rotation Contribution=0, Translation Contribution=0) {module=BlendSpace001}
   &OffsetPosition(
    Position Offset=&MakeVector(
     X=0
     Y=&FloatFromCurve(
      FloatCurve=DataInterface() {propertyValues {Curve {
         defaultValue=1
         keys [{interpMode=RCIM_Cubic, tangentMode=RCTM_User, time=0, value=0}, {interpMode=RCIM_Cubic, tangentMode=RCTM_User, time=1, value=1}]
        }}}
      Scale Curve=&RandomRangeFloat(Minimum=-15, Maximum=-25)
     )
     Z=0
    )
    Position Offset Coordinate Space:ENiagaraCoordinateSpace=::Local
   )
  ]
  rendererClasses [&NiagaraSpriteRendererProperties]
  renderers [
   {rendererClass=&NiagaraSpriteRendererProperties}
   {rendererClass=&NiagaraSpriteRendererProperties, rendererIndex=1}
  ]
 }
 carpet_glint_old {
  simTarget=GPUComputeSim
  emitterSpawnScript [/SetVariables(Emitter.FE Particle Spawn Rate=60, Emitter.Idle Spawn Rate=200, Emitter.Max Traversal Speed=150, Emitter.Particle Spawn Rate=60) {isSetParameters=true, module=SetVariables_#2}]
  emitterUpdateScript [
   &EmitterState()
   &SpawnBurst_Instantaneous(Spawn Group=2, Use Spawn Probability=false, UseLoopCountLimit=false)
   &SpawnRate(
    SpawnRate=&FN_FrontEndFloatModifier(
     Gameplay=&MakeCustomFloatFromBool(
      Bool=&SetBoolByFloatComparison(A=&VectorLength(Vector=$User.OwnerLocalVelocity), B=140)
      True Float=$Emitter.Particle Spawn Rate
      False Float=$Emitter.Idle Spawn Rate
     )
     Front End=$Emitter.FE Particle Spawn Rate
    )
    Use Spawn Probability=false
   ) {module=SpawnRate001}
  ]
  particleSpawnScript [
   /SetVariables(Particles.VisibilityTag=&RandomRangeInt(Minimum=0, Maximum=2)) {isSetParameters=true, module=SetVariables_#3}
   &InitializeParticle(
    AdjustAlpha=false
    AdjustHue=true
    AdjustSaturation=false
    AdjustValue=false
    Color=<rgba 0.843 0.8 1 1>
    Color Mode=::NewEnumerator3
    Hue Shift Range:Vector2f=<-1 1>
    Lifetime Max=0.85
    Lifetime Min=0.5
    Lifetime Mode=::NewEnumerator1
    Sprite Rotation Mode=::NewEnumerator2
    Sprite Size Mode=::NewEnumerator3
    Uniform Sprite Size=&Multiply_Float(
     A=&Multiply_Float(
      A=&SelectFloatFromArray(Direct Array Index=$Particles.VisibilityTag, Float Selection Array=DataInterface() {propertyValues {FloatData [100,30,15]}})
      B=&RandomRangeFloat(Minimum=0.55, Maximum=0.8)
     )
     B=1
    )
    UsePositionOffset=true
    Write Material Random=true
   )
   &CameraOffset(Camera Offset Amount=10)
   &SkeletalMeshLocation(Bone Sampling Mode=::NewEnumerator4, MaskSpawnGroup=false, Sampled Position Offset=<0 20 15>, Skeletal Mesh=$System.Skeletal Mesh)
   &ShapeLocation(
    Box Size=<95 145 5>
    MaskSpawnGroup=false
    Offset Mode=::NewEnumerator0
    Override Seed=false
    Shape Primitive=::NewEnumerator2
    Surface Only=false
    UseOwnerScale=true
   )
   &AddVelocity(Velocity=&RandomRangeVector(Minimum=<0 -100 0>, Maximum=<1 -350 1>))
  ]
  particleUpdateScript [
   &ParticleState()
   &ScaleColor(
    Scale Alpha=&Multiply_Float(
     A=&FloatFromCurve(
      FloatCurve=DataInterface() {propertyValues {Curve {keys [
          {interpMode=RCIM_Cubic, tangentMode=RCTM_User, time=0, value=0}
          {arriveTangent=-2.4058, interpMode=RCIM_Cubic, leaveTangent=-2.4058, tangentMode=RCTM_User, time=0.26, value=0.8421}
          {arriveTangent=-0.9001, interpMode=RCIM_Cubic, leaveTangent=-0.9001, time=0.4613, value=0.0645}
          {arriveTangent=1.9066, interpMode=RCIM_Cubic, leaveTangent=1.9066, time=0.5594, value=0.5726}
          {arriveTangent=-0.3157, interpMode=RCIM_Cubic, leaveTangent=-0.3157, tangentMode=RCTM_User, time=0.6774, value=0.4764}
          {arriveTangent=-7.1699, interpMode=RCIM_Cubic, leaveTangent=-7.1699, tangentMode=RCTM_User, time=1, value=0}
         ]}}}
      Scale Curve=&RandomRangeFloat(Minimum=0.1, Maximum=0.5)
     )
     B=&FloatFromCurve(
      FloatCurve=DataInterface() {propertyValues {Curve {
         defaultValue=0
         keys [{interpMode=RCIM_Cubic, tangentMode=RCTM_User, time=0, value=0}, {interpMode=RCIM_Cubic, tangentMode=RCTM_User, time=1, value=1}]
        }}}
      CurveIndex=&Multiply_Float(
       A=&MakeFloatFromVector(Vector=Linked() {linkedVariable {name=Particles.Position, type=NiagaraPosition}}, Channel=::NewEnumerator2)
       B=$Engine.System.Age
      )
     )
    )
    Scale RGB=&MakeVectorFromLinearColorRGB(Color=&FNTimeOfDayColor(Night Color=<rgba 0.3 0.6267 1 5>, Morning Color=<rgba 1 1 1 1>, Day Color=<rgba 5 5 5 1>, Evening Color=<rgba 1 1 1 1>))
    ScaleA=true
    ScaleRGB=true
   )
   &ScaleSpriteSize(Scale Factor=&Multiply_Vector2D_ByFloat(
     Float=1
     Vector2D=&Vector2DFromFloat(Value=&FloatFromCurve(
       FloatCurve=DataInterface() {propertyValues {Curve {keys [
           {arriveTangent=9.0592, interpMode=RCIM_Cubic, leaveTangent=9.0592, tangentMode=RCTM_User, time=0, value=0}
           {interpMode=RCIM_Cubic, tangentMode=RCTM_User, time=0.2, value=1}
           {tangentMode=RCTM_User, time=1, value=0}
          ]}}}
       Scale Curve=&RandomRangeFloat(Minimum=0.3, Maximum=1)
      ))
    ))
   /SetVariables(
    Particles.Color=&MultiplyLinearColor(
     Linear Color=$Particles.Color
     Scale Factor=&MakeVector4_FromLinearColor(LinearColor=<rgba 8.0099 9.8007 10 1>)
    )
    Particles.Color02:LinearColor=&MultiplyLinearColor(
     Linear Color=$Particles.Color
     Scale Factor=&MakeVector4_FromLinearColor(LinearColor=<rgba 0.9441 0.8079 1 1>)
    )
    Particles.Color03:LinearColor=&MultiplyLinearColor(
     Linear Color=$Particles.Color
     Scale Factor=&MakeVector4_FromLinearColor(LinearColor=<rgba 0.15 0.15 0.15 1>)
    )
    Particles.SpriteSize02:Vector2f=&Multiply_Vector2(A=$Particles.SpriteSize, B=&Vector2DFromFloat(Value=0.3))
   ) {isSetParameters=true, module=SetVariables_#4}
   &CurlNoiseForce(Noise Strength=100, Pan Noise=false)
   &Drag(
    Drag=&FloatFromCurve(
     FloatCurve=DataInterface() {propertyValues {Curve {
        defaultValue=0
        keys [{interpMode=RCIM_Cubic, tangentMode=RCTM_User, time=0, value=1}, {interpMode=RCIM_Cubic, tangentMode=RCTM_User, time=1, value=0}]
       }}}
     Scale Curve=&RandomRangeFloat(Minimum=10, Maximum=12)
    )
    Use Linear Drag=true
    Use Rotational Drag=true
   ) {module=Drag001}
   &SolveForcesAndVelocity(Clamp Velocity=false, Limit Acceleration=false)
   &BlendSpace("Position Blend (World <-> Local)":NiagaraFloat=&MakeCustomFloatFromBool(
     Bool=&SetBoolByFloatComparison(A=&VectorLength(Vector=$Engine.Owner.Velocity), B=400)
     True Float=1
     False Float=&FloatFromCurve(FloatCurve=DataInterface() {propertyValues {Curve {keys [<at 0 1 RCIM_Cubic>, {arriveTangent=-12.1775, interpMode=RCIM_Cubic, leaveTangent=-12.1775, tangentMode=RCTM_User, time=0.25, value=0}]}}})
     Translation Contribution:NiagaraFloat=1
     Rotation Contribution:NiagaraFloat=0
    )) {module=BlendSpace001}
   &FN_Light_Attributes(
    Color=<rgba 1 0.8484 0.5817 1>
    Create Color Attribute=true
    Create Position Attribute=true
    Create Radius Attribute=true
    Enable Particle Light Enable=true
    Enable Visibility Tag=true
    Light Enabled=true
    Position=&AddVectorToPosition(Position=$Particles.SkeletalMeshLocation001.SampledPosition, Vector=<0 0 50>)
    Radius=50
    Scale Color=50
    VisibilityTag=2
   ) {enabled=false}
  ]
  rendererClasses [&NiagaraSpriteRendererProperties]
  renderers [
   {rendererClass=&NiagaraSpriteRendererProperties}
   {rendererClass=&NiagaraSpriteRendererProperties, rendererIndex=1}
  ]
 }
}
summary {
 systemName=NS_Emote_RowLiaison_FE_Traversal_Sparkles
 emitters [
  {emitterName=Body_Glitter, rendererClasses [&NiagaraSpriteRendererProperties]}
  {emitterName=carpet_glint_old, simTarget=GPUComputeSim, rendererClasses [&NiagaraSpriteRendererProperties]}
  {emitterName=Light, rendererClasses [&NiagaraLightRendererProperties]}
  {emitterName=Glow_on_feet__filler_for_, simTarget=GPUComputeSim, rendererClasses [&NiagaraSpriteRendererProperties]}
  {emitterName=Med_sparkles, simTarget=GPUComputeSim, rendererClasses [&NiagaraSpriteRendererProperties]}
  {emitterName=Carpet_Glint_Low-Med, rendererClasses [&NiagaraSpriteRendererProperties]}
  {emitterName=Glow_on_feet_Low-_Med, rendererClasses [&NiagaraSpriteRendererProperties]}
 ]
 userVariables [
  User.bIsFrontEnd {type=NiagaraBool}
  User.bIsFrontEndPreview {type=NiagaraBool}
  User.OwnerLocalVelocity {type=Vector3f, description="", defaultValue {struct=Vector3f, value=<0 0 0>}}
 ]
}
