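# Niagara system for the RowLiaison emote, producing a brief socket-attached flash with pink heart and soft glow sprites plus a light.
# Pink_heart001 fires two instantaneous bursts whose spawn group switches color, lifetime, size and mesh index.
# Sampled socket positions feed offset heart positions and a particle light, with curves shaping fades, growth and time-of-day-scaled brightness.
# Flash and FlashStar show no modules or renderers here; user variables expose front-end flags and notify progress.
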NiagaraSystem /BRCosmetics/Animation/Game/MainPlayer/Emotes/RowLiaison/FX/NS_Emote_RowLiaison_Flash_LWE
dependencies {
 usedDataInterfaces [
  &NiagaraDataInterfaceCurve
  &NiagaraDataInterfaceVector2DCurve
  &NiagaraDataInterfaceSocketReader
 ]
 usedDynamicInputs [
  &FN_FrontEndVector3Modifier
  &Multiply_Vector2D_ByFloat
  &FN_FrontEndFloatModifier
  &RandomRangeFloat
  &MakeCustomIntFromBool
 ]
 usedModules [
  &self:SystemSpawnScript.SystemScriptSource.SystemScriptGraph.NiagaraNodeAssignment_0.SetVariables_#1
  &self:SystemSpawnScript.SystemScriptSource.SystemScriptGraph.NiagaraNodeAssignment_1.SetVariables_#2
  &self:SystemSpawnScript.SystemScriptSource.SystemScriptGraph.NiagaraNodeAssignment_2.SetVariables_#3
  &self:SystemSpawnScript.SystemScriptSource.SystemScriptGraph.NiagaraNodeAssignment_3.SetVariables_#4
  &self:SystemSpawnScript.SystemScriptSource.SystemScriptGraph.NiagaraNodeAssignment_4.SetVariables_#5
  &SystemState
 ]
}
emitters {
 Flash {rendererClasses [], renderers []}
 FlashStar {rendererClasses [], renderers []}
 Pink_heart001 {
  emitterUpdateScript [
   &EmitterState()
   &SpawnBurst_Instantaneous(Spawn Group=1, Use Spawn Probability=false, UseLoopCountLimit=false) {module=SpawnBurst_Instantaneous001}
   &SpawnRate(SpawnRate=25, Use Spawn Probability=false) {enabled=false}
   &SpawnBurst_Instantaneous(Spawn Count=5, Use Spawn Probability=false, UseLoopCountLimit=false)
  ]
  particleSpawnScript [
   /SetVariables(Particles.CameraOffset=10, Particles.VisibilityTag=&SpawnGroup()) {isSetParameters=true}
   &InitializeParticle(
    Color=&MergeHelper_ColorSwitch(0=<rgba 15 1.6424 7.108 2>, 1=<rgba 1 1 1 1>, 2=?, 3=?, Index=&SpawnGroup(), Num=::NewEnumerator0)
    Lifetime=&MergeHelper_FloatSwitch(0=0.4, 1=0.5, 2=?, 3=?, Index=&SpawnGroup(), Num=::NewEnumerator0)
    Mesh Scale Mode=::NewEnumerator3
    Sprite Size Mode=::NewEnumerator3
    Uniform Sprite Size=&MergeHelper_FloatSwitch(0=25, 1=30, 2=?, 3=?, Index=&SpawnGroup(), Num=::NewEnumerator0)
    UsePositionOffset=false
    Write Material Random=true
   )
   &SpriteFacingAndAlignment(Alignment Coordinate Space=::Local, Sprite Facing=<0 0 1>, WriteAlignment=true, WriteFacing=true) {enabled=false}
   /SetVariables(Particles.MeshIndex=&SpawnGroup()) {isSetParameters=true}
  ]
  particleUpdateScript [
   &ParticleState()
   &ScaleSpriteSize(
    Uniform Curve Scale=1.3
    Uniform Curve Sprite Scale=DataInterface() {propertyValues {Curve {
       defaultValue=0
       keys [<at 0 0 RCIM_Cubic>, {arriveTangent=2.5, interpMode=RCIM_Cubic, leaveTangent=2.5, time=0.2474, value=1.3294}, <at 1 2.5>]
      }}}
   ) {module=ScaleSpriteSize001}
   /SetVariables(
    Particles.soft glow size:Vector2f=&Vector2DFromFloat(Value=&FloatFromCurve(
      FloatCurve=DataInterface() {propertyValues {Curve {
         defaultValue=0
         keys [<at 0 0 RCIM_Cubic>, {arriveTangent=2.5, interpMode=RCIM_Cubic, leaveTangent=2.5, time=0.2474, value=1.3294}, <at 1 2.5>]
        }}}
      Scale Curve=34
     ))
    Particles.soft glowLinear Color:LinearColor=&MakeLinearColorFromVectorAndFloat(
     "Vector (RGB)"=&Multiply_VectorByFloat(
      Vector=&MakeVectorFromLinearColorRGB(Color=$Particles.Color)
      Float=&FloatFromCurve(FloatCurve=DataInterface() {propertyValues {Curve {
          defaultValue=0.5
          keys [
           {arriveTangent=9.0725, interpMode=RCIM_Cubic, leaveTangent=9.0725, tangentMode=RCTM_User, time=0, value=0}
           {arriveTangent=-1.9266, interpMode=RCIM_Cubic, leaveTangent=-1.9266, tangentMode=RCTM_User, time=0.1408, value=0.9847}
           {arriveTangent=-0.9094, interpMode=RCIM_Cubic, leaveTangent=-0.9094, tangentMode=RCTM_User, time=1.5, value=0}
          ]
         }}})
     )
     "Float (Alpha)"=&FloatFromCurve(FloatCurve=DataInterface() {propertyValues {Curve {
         defaultValue=1
         keys [<at 0 0 RCIM_Cubic>, {arriveTangent=-0.0126, interpMode=RCIM_Cubic, leaveTangent=-0.0126, time=0.1, value=1}, <at 0.75 -0.0094 RCIM_Cubic>]
        }}})
    )
   ) {isSetParameters=true}
   &SocketLocation(
    SamplePosOffset=true
    Sampled Position Offset=<0 0 5>
    Socket Reader=$System.Socket Reader
    Socket Sampling Mode=::NewEnumerator3
    VelocitySpeedLimit=false
   )
   /SetVariables(
    Particles.Heart Color:LinearColor=&MakeLinearColorFromVectorAndFloat(
     "Vector (RGB)"=&Multiply_VectorByFloat(
      Vector=&MakeVectorFromLinearColorRGB(Color=$Particles.Color)
      Float=&FloatFromCurve(
       FloatCurve=DataInterface() {propertyValues {Curve {
          defaultValue=0.5
          keys [
           {arriveTangent=9.0725, interpMode=RCIM_Cubic, leaveTangent=9.0725, tangentMode=RCTM_User, time=0, value=0}
           {arriveTangent=-1.9266, interpMode=RCIM_Cubic, leaveTangent=-1.9266, tangentMode=RCTM_User, time=0.1408, value=0.9847}
           {arriveTangent=-0.3934, interpMode=RCIM_Cubic, leaveTangent=-0.3934, tangentMode=RCTM_User, time=1.5, value=0}
          ]
         }}}
       Scale Curve=3
      )
     )
     "Float (Alpha)"=&FloatFromCurve(
      FloatCurve=DataInterface() {propertyValues {Curve {
         defaultValue=1
         keys [<at 0 0 RCIM_Cubic>, {arriveTangent=-0.0151, interpMode=RCIM_Cubic, leaveTangent=-0.0151, time=0.1, value=1}, <at 0.6249 -0.0094 RCIM_Cubic>]
        }}}
      Scale Curve=0.2
     )
    )
    Particles.Heart SpriteFacing:Vector3f=<-0.5 1 0.9088>
    Particles.Heart position:NiagaraPosition=&AddVectorToPosition(
     Position=$Output.SocketLocation.SampledPosition
     Vector=&MakeVector(
      X=10
      Y=&FloatFromCurve(FloatCurve=DataInterface() {propertyValues {Curve {keys [<at 0 1>,<at 0.5 -15>]}}})
      Z=&FloatFromCurve(FloatCurve=DataInterface() {propertyValues {Curve {keys [<at 0 1>,<at 0.5 -5>]}}})
     )
    )
    Particles.heart position 2:NiagaraPosition=&AddVectorToPosition(Position=$Output.SocketLocation.SampledPosition, Vector=<0 0 0>)
   ) {isSetParameters=true}
   /SetVariables() {isSetParameters=true}
   &FN_Light_Attributes(
    Color=<rgba 1 0.941 0.6882 10>
    Create Color Attribute=true
    Create Exponent Attribute=true
    Create Position Attribute=true
    Create Radius Attribute=true
    Enable Particle Light Enable=true
    Exponent=6
    Light Enabled=true
    Position=&AddVectorToPosition(Position=$Output.SocketLocation.SampledPosition, Vector=<9.9524 0 -25>)
    Radius=&FloatFromCurve(
     FloatCurve=DataInterface() {propertyValues {Curve {
        defaultValue=0
        keys [
         {arriveTangent=9.0725, interpMode=RCIM_Cubic, leaveTangent=9.0725, tangentMode=RCTM_User, time=0, value=0}
         {arriveTangent=0.2697, interpMode=RCIM_Constant, tangentMode=RCTM_User, time=0.3, value=0.9847}
         {arriveTangent=-1.4067, interpMode=RCIM_Cubic, leaveTangent=-1.4067, time=0.7, value=1}
         {arriveTangent=-0.0405, interpMode=RCIM_Cubic, leaveTangent=-0.0405, tangentMode=RCTM_User, time=1, value=0}
        ]
       }}}
     Scale Curve=75
    )
    Scale Color=&FloatFromCurve(
     FloatCurve=DataInterface() {propertyValues {Curve {
        defaultValue=0
        keys [
         {arriveTangent=9.0725, interpMode=RCIM_Cubic, leaveTangent=9.0725, tangentMode=RCTM_User, time=0, value=0}
         {arriveTangent=0.2697, interpMode=RCIM_Constant, tangentMode=RCTM_User, time=0.3, value=0.9847}
         {arriveTangent=-1.4067, interpMode=RCIM_Cubic, leaveTangent=-1.4067, time=0.4898, value=1}
         {arriveTangent=-0.0405, interpMode=RCIM_Cubic, leaveTangent=-0.0405, tangentMode=RCTM_User, time=1, value=0}
        ]
       }}}
     Scale Curve=&FNTimeOfDayScalar(Night Scale=1.3, Morning Scale=1, Day Scale=1.5, Evening Scale=1)
    )
   )
  ]
  rendererClasses [&NiagaraSpriteRendererProperties, &NiagaraLightRendererProperties]
  renderers [
   {rendererClass=&NiagaraSpriteRendererProperties}
   {rendererClass=&NiagaraLightRendererProperties, rendererIndex=1}
   {rendererClass=&NiagaraSpriteRendererProperties, rendererIndex=2}
  ]
 }
}
summary {
 systemName=NS_Emote_RowLiaison_Flash_LWE
 emitters [
  {emitterName=Flash, rendererClasses []}
  {emitterName=FlashStar, rendererClasses []}
  {emitterName=Pink_heart001, rendererClasses [&NiagaraSpriteRendererProperties, &NiagaraLightRendererProperties]}
 ]
 userVariables [
  User.bIsFrontEnd {type=NiagaraBool}
  User.bIsFrontEndPreview {type=NiagaraBool}
  User.NormalizedNotifyProgress {type=NiagaraFloat, defaultValue=0, description=""}
 ]
}
