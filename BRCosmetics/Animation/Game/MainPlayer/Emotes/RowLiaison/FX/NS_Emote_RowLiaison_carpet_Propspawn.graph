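# Niagara system for the RowLiaison emote, named for the carpet prop spawn, that fires a one-shot burst of glowing spark sprites.
# A single emitter spawns sparks on a disc and pushes them outward, then curl noise, fading drag and a growing upward gravity force move them.
# Curves drive size and flickering alpha, with front-end/gameplay scaling and time-of-day tint.
# Secondary color and size attributes accompany two sprite renderers.
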NiagaraSystem /BRCosmetics/Animation/Game/MainPlayer/Emotes/RowLiaison/FX/NS_Emote_RowLiaison_carpet_Propspawn
dependencies {
 usedDataInterfaces [&NiagaraDataInterfaceArrayFloat, &NiagaraDataInterfaceCurve, &NiagaraDataInterfaceVector2DCurve]
 usedDynamicInputs []
 usedModules [&SystemState]
}
emitters {burst_around_the_star_sparks001 {
  emitterUpdateScript [
   &EmitterState()
   &SpawnBurst_Instantaneous(Spawn Count=&UniformRangedInt(Minimum=64, Maximum=100), Spawn Time=0.05, Use Spawn Probability=false)
   &SpawnRate(SpawnRate=50, Use Spawn Probability=false) {enabled=false}
  ]
  particleSpawnScript [
   &InitializeParticle(
    Color=<rgba 3.63 4.1486 5 1>
    Lifetime Max=0.85
    Lifetime Min=0.3
    Lifetime Mode=::NewEnumerator1
    Position Offset=<0 0 1.5>
    Sprite Rotation Mode=::NewEnumerator2
    Sprite Size Mode=::NewEnumerator3
    Uniform Sprite Size=&Multiply_Float(
     A=&Multiply_Float(
      A=&SelectFloatFromArray(Direct Array Index=$Particles.VisibilityTag, Float Selection Array=DataInterface() {propertyValues {FloatData [100,30,15]}})
      B=&RandomRangeFloat(Minimum=0.4, Maximum=0.7)
     )
     B=1
    )
    UsePositionOffset=true
    Write Material Random=true
   )
   &ShapeLocation(
    Disc Coverage=0.7667
    MaskSpawnGroup=false
    Override Seed=false
    Ring / Disc Mode:ENiagaraRingDiscMode=::NewEnumerator1
    Ring Radius=10
    Shape Primitive=::NewEnumerator4
    U Distribution=0.7857
    UseOwnerScale=true
   )
   &AddVelocityFromPoint(
    Module.Use Falloff Distance=false
    Origin Offset=&MakeVector(X=0, Y=0, Z=&RandomRangeFloat(Minimum=-3, Maximum=10))
    Velocity Strength=&Multiply_Float(A=&RandomRangeFloat(Minimum=0.5, Maximum=1), B=850)
   )
   &ApplyInitialForces()
  ]
  particleUpdateScript [
   &ParticleState()
   &ScaleSpriteSize(Scale Factor=&Multiply_Vector2D_ByFloat(
     Float=1
     Vector2D=&Vector2DFromFloat(Value=&FloatFromCurve(
       FloatCurve=DataInterface() {propertyValues {Curve {keys [
           {arriveTangent=9.0592, interpMode=RCIM_Cubic, leaveTangent=9.0592, tangentMode=RCTM_User, time=0, value=0}
           {interpMode=RCIM_Cubic, tangentMode=RCTM_User, time=0.2, value=1}
           {tangentMode=RCTM_User, time=1, value=0}
          ]}}}
       Scale Curve=&FN_FrontEndFloatModifier(
        Gameplay=&RandomRangeFloat(Minimum=0.35, Maximum=0.8)
        Front End=&RandomRangeFloat(Minimum=0.5, Maximum=0.7)
       )
      ))
    )) {module=ScaleSpriteSize001}
   &ScaleSpriteSize(Scale Factor=&Vector2DFromCurve(
     Vector2Curve=DataInterface() {propertyValues {
       XCurve {keys [<at 0 1 RCIM_Cubic>, {arriveTangent=-0.5084, interpMode=RCIM_Cubic, leaveTangent=-0.5084, tangentMode=RCTM_User, time=1, value=0}]}
       YCurve {keys [{arriveTangent=-3.0155, interpMode=RCIM_Cubic, leaveTangent=-3.0155, tangentMode=RCTM_User, time=0, value=1}, <at 1 0 RCIM_Cubic>]}
      }}
     ScaleCurve=&Vector2DFromFloat(Value=&Lerp_Float(
       A=&FloatFromCurve(
        FloatCurve=DataInterface() {propertyValues {Curve {keys [
            {interpMode=RCIM_Cubic, tangentMode=RCTM_Break, time=0, value=0.5}
            {arriveTangent=0.001, interpMode=RCIM_Cubic, tangentMode=RCTM_Break, time=1000, value=1}
           ]}}}
        CurveIndex=$Engine.Owner.LODDistance
       )
       B=1
       Alpha=0.5
      ))
    )) {enabled=false}
   &CurlNoiseForce(Noise Frequency=5, Noise Strength=500, Pan Noise=true, Pan Noise Field=<0 0.2 0.5>)
   &Drag(
    Drag=&FloatFromCurve(
     FloatCurve=DataInterface() {propertyValues {Curve {keys [<at 0 1>,<at 0.5 0>]}}}
     Scale Curve=&UniformRangedFloat(Minimum=5, Maximum=15)
    )
    Use Linear Drag=true
    Use Rotational Drag=false
   )
   &GravityForce(Gravity=&MakeVector(
     X=0
     Y=0
     Z=&FloatFromCurve(
      FloatCurve=DataInterface() {propertyValues {Curve {
         defaultValue=1
         keys [{interpMode=RCIM_Cubic, tangentMode=RCTM_User, time=0, value=0}, {interpMode=RCIM_Cubic, tangentMode=RCTM_User, time=1, value=0.5}]
        }}}
      Scale Curve=&RandomRangeFloat(Minimum=300, Maximum=980)
     )
    ))
   &SolveForcesAndVelocity(Clamp Velocity=false, Limit Acceleration=false)
   &ScaleColor(
    Scale Alpha=&Multiply_Float(
     A=&FloatFromCurve(
      FloatCurve=DataInterface() {propertyValues {Curve {keys [
          {interpMode=RCIM_Cubic, tangentMode=RCTM_User, time=0, value=0}
          {arriveTangent=-2.4058, interpMode=RCIM_Cubic, leaveTangent=-2.4058, tangentMode=RCTM_User, time=0.26, value=0.8421}
          {arriveTangent=-1.1775, interpMode=RCIM_Cubic, leaveTangent=-1.1775, time=0.4613, value=0.0645}
          {arriveTangent=-0.1197, interpMode=RCIM_Cubic, leaveTangent=-0.1197, time=0.5898, value=0.4537}
          <at 1 0 RCIM_Cubic>
         ]}}}
      Scale Curve=&RandomRangeFloat(Minimum=0.1, Maximum=0.5)
     )
     B=&FloatFromCurve(
      FloatCurve=DataInterface() {propertyValues {Curve {keys [{tangentMode=RCTM_User, time=0, value=0}, {tangentMode=RCTM_User, time=1, value=1}]}}}
      CurveIndex=&Multiply_Float(
       A=&MakeFloatFromVector(Vector=Linked() {linkedVariable {name=Particles.Position, type=NiagaraPosition}}, Channel=::NewEnumerator2)
       B=$Engine.System.Age
      )
     )
    )
    Scale RGB=&MakeVectorFromLinearColorRGB(Color=&FNTimeOfDayColor(Night Color=<rgba 0.9 1.88 3 5>, Morning Color=<rgba 1 1 1 1>, Day Color=<rgba 5 5 5 1>, Evening Color=<rgba 1 1 1 1>))
    ScaleA=true
    ScaleRGB=true
   )
   /SetVariables(
    Particles.Color=&MultiplyLinearColor(
     Linear Color=$Particles.Color
     Scale Factor=&MakeVector4_FromLinearColor(LinearColor=<rgba 8.0099 9.8007 10 1>)
    )
    Particles.Color02:LinearColor=&MultiplyLinearColor(
     Linear Color=$Particles.Color
     Scale Factor=&MakeVector4_FromLinearColor(LinearColor=<rgba 0.9441 0.8079 1 1>)
    )
    Particles.Color03:LinearColor=&MultiplyLinearColor(
     Linear Color=$Particles.Color
     Scale Factor=&MakeVector4_FromLinearColor(LinearColor=<rgba 0.15 0.15 0.15 1>)
    )
    Particles.SpriteSize02:Vector2f=&Multiply_Vector2(A=$Particles.SpriteSize, B=&Vector2DFromFloat(Value=0.3))
   ) {isSetParameters=true}
  ]
  rendererClasses [&NiagaraSpriteRendererProperties]
  renderers [
   {rendererClass=&NiagaraSpriteRendererProperties}
   {rendererClass=&NiagaraSpriteRendererProperties, rendererIndex=1}
  ]
 }}
summary {
 systemName=NS_Emote_RowLiaison_carpet_Propspawn
 emitters [{emitterName=burst_around_the_star_sparks001, rendererClasses [&NiagaraSpriteRendererProperties]}]
 userVariables [
  User.bIsFrontEnd {type=NiagaraBool}
  User.bIsFrontEndPreview {type=NiagaraBool}
  User.NormalizedNotifyProgress {type=NiagaraFloat, defaultValue=0, description=""}
  User.OwnerLocalVelocity {type=Vector3f, description="", defaultValue {struct=Vector3f, value=<0 0 0>}}
 ]
}
