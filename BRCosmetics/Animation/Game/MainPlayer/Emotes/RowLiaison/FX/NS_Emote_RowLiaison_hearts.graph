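# Niagara effect for the RowLiaison emote that bursts heart and spark sprites.
# One emitter fires two bursts, with SpawnGroup switches giving a larger red sprite upward velocity and squash-stretch, and ten smaller sprites point force, curl noise and drag.
# Spawn group also feeds a VisibilityTag alongside three sprite renderers.
# Curves drive fading, small-sprite flicker and custom glow color, with time-of-day scaling brightening small sprites at night.
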NiagaraSystem /BRCosmetics/Animation/Game/MainPlayer/Emotes/RowLiaison/FX/NS_Emote_RowLiaison_hearts
dependencies {
 usedDataInterfaces [
  &NiagaraDataInterfaceCurve
  &NiagaraDataInterfaceSpriteRendererInfo
  &NiagaraDataInterfaceVector2DCurve
 ]
 usedDynamicInputs [&MakeFloatFromVector2d]
 usedModules [&SystemState]
}
emitters {Sparks_and_Hearts003 {
  emitterUpdateScript [
   &EmitterState()
   &SpawnBurst_Instantaneous(Use Spawn Probability=false, UseLoopCountLimit=false)
   &SpawnBurst_Instantaneous(Spawn Count=10, Spawn Group=1, Use Spawn Probability=false, UseLoopCountLimit=false) {module=SpawnBurst_Instantaneous001}
  ]
  particleSpawnScript [
   /SetVariables(Particles.VisibilityTag=&SpawnGroup()) {isSetParameters=true}
   &InitializeParticle(
    Color=&MergeHelper_ColorSwitch(
     0=<rgba 0.3656 0 0.0059 1>
     1=&RandomRangeLinearColor(Minimum=<rgba 2 0 0.0324 1>, Maximum=<rgba 0.1229 0.0019 0 1>)
     2=?
     3=?
     Index=&SpawnGroup()
     Num=::NewEnumerator0
    )
    Lifetime=&MergeHelper_FloatSwitch(
     0=0.75
     1=&RandomRangeFloat(Minimum=0.5, Maximum=1)
     2=?
     3=?
     Index=&SpawnGroup()
     Num=::NewEnumerator0
    )
    Sprite Size Mode=::NewEnumerator3
    Sprite UV Mode=::NewEnumerator1
    Uniform Sprite Size=&MergeHelper_FloatSwitch(
     0=45
     1=&RandomRangeFloat(Minimum=15, Maximum=25)
     2=?
     3=?
     Index=&SpawnGroup()
     Num=::NewEnumerator0
    )
    UsePositionOffset=false
    Write Material Random=true
   )
   &AddVelocity(Velocity=&MergeHelper_VectorSwitch(
     0=&MakeVector(
      X=0
      Y=0
      Z=&FloatFromCurve(
       FloatCurve=DataInterface() {propertyValues {Curve {
          defaultValue=1
          keys [
           {interpMode=RCIM_Cubic, tangentMode=RCTM_User, time=0.0051, value=0.6868}
           {interpMode=RCIM_Cubic, tangentMode=RCTM_User, time=0.0963, value=0.0624}
          ]
         }}}
       Scale Curve=100
      )
     )
     1=&RandomRangeVector(Minimum=<-25 -25 -25>, Maximum=<25 25 25>)
     2=<0 0 0>
     3=<0 0 0>
     4=<0 0 0>
     Index=&SpawnGroup()
     Num=::NewEnumerator3
    ))
   &SubUVAnimation(
    Evaluation Type=::NewEnumerator0
    "Sprite Renderer "=DataInterface() {propertyValues {SpriteRenderer=&self:Sparks_and_Hearts003_0.NiagaraSpriteRendererProperties_0}}
    SubUV Animation Mode:ENiagara_SubUVLookupModeV2=::NewEnumerator2
    UseEndFrame=false
    UseStartFrame=false
   ) {module=SubUVAnimation001}
   &ShapeLocation(MaskSpawnGroup=false, Override Seed=false, Sphere Radius=5, UseOwnerScale=true)
  ]
  particleUpdateScript [
   &ParticleState()
   &SpriteRotationRate(Rotation Rate=&FloatFromCurve(
     FloatCurve=DataInterface() {propertyValues {Curve {
        defaultValue=0
        keys [
         {arriveTangent=1.8448, interpMode=RCIM_Cubic, leaveTangent=1.8448, tangentMode=RCTM_User, time=0, value=0}
         {interpMode=RCIM_Cubic, tangentMode=RCTM_User, time=1, value=1}
        ]
       }}}
     Scale Curve=&MaskFloatBySpawnGroup(
      Masked Float=&RandomRangeFloat(Minimum=-200, Maximum=200)
      Passthrough Float=&RandomRangeFloat(Minimum=-500, Maximum=500)
     )
    ))
   &ScaleSpriteSize(
    Scale Factor=&MergeHelper_Vector2Switch(
     0=&Vector2DFromCurve(
      Vector2Curve=DataInterface() {propertyValues {
        XCurve {keys [{tangentMode=RCTM_User, time=0, value=1}, {tangentMode=RCTM_User, time=1, value=1}]}
        YCurve {keys [
          <at 0 1 RCIM_Cubic>
          {arriveTangent=-3.3059, interpMode=RCIM_Cubic, leaveTangent=-3.3059, time=0.0398, value=3}
          {arriveTangent=-10.7205, interpMode=RCIM_Cubic, leaveTangent=-10.7205, time=0.0989, value=0.673}
          {arriveTangent=0.2979, interpMode=RCIM_Cubic, leaveTangent=0.2979, tangentMode=RCTM_User, time=0.1576, value=1.7365}
          {arriveTangent=-1.3261, interpMode=RCIM_Cubic, leaveTangent=-1.3261, tangentMode=RCTM_User, time=0.2002, value=0.9659}
          {arriveTangent=0.0697, interpMode=RCIM_Cubic, leaveTangent=0.0697, time=0.3237, value=1}
          <at 0.5255 0.9886 RCIM_Cubic>
          {arriveTangent=0.0241, interpMode=RCIM_Cubic, leaveTangent=0.0241, time=0.8748, value=1}
          <at 1 1 RCIM_Cubic>
         ]}
       }}
      ScaleCurve=<1 1>
     )
     1=&Vector2DFromFloat(Value=&FloatFromCurve(
       FloatCurve=DataInterface() {propertyValues {Curve {
          defaultValue=1
          keys [{interpMode=RCIM_Cubic, tangentMode=RCTM_Break, time=0, value=0}, {interpMode=RCIM_Cubic, tangentMode=RCTM_Break, time=0.5, value=1}]
         }}}
       Scale Curve=2
      ))
     2=?
     3=?
     4=?
     Index=&SpawnGroup()
     Num=::NewEnumerator0
    )
    Scale Sprite Size Mode=::NewEnumerator2
   )
   &ScaleSpriteSize(
    Scale Sprite Size Mode=::NewEnumerator0
    Uniform Scale Factor=&FloatFromCurve(FloatCurve=DataInterface() {propertyValues {Curve {
        defaultValue=0
        keys [
         {arriveTangent=9.0725, interpMode=RCIM_Cubic, leaveTangent=9.0725, tangentMode=RCTM_User, time=0, value=0}
         {arriveTangent=-0.9362, interpMode=RCIM_Cubic, leaveTangent=-0.9362, tangentMode=RCTM_User, time=0.1408, value=0.9847}
         {arriveTangent=-1.146, interpMode=RCIM_Cubic, leaveTangent=-1.146, time=0.2037, value=0.9509}
         {arriveTangent=-0.2621, interpMode=RCIM_Cubic, leaveTangent=-0.2621, tangentMode=RCTM_User, time=1, value=0}
        ]
       }}})
   ) {module=ScaleSpriteSize001}
   &PointForce(
    Force Strength=&MergeHelper_FloatSwitch(0=0, 1=1000, 2=0, 3=0, 4=0, Index=&SpawnGroup(), Num=::NewEnumerator3)
    Module.Use Falloff=true
    Module.UseLinearForce=false
   )
   &CurlNoiseForce(
    Bias Noise=false
    Noise Frequency=15
    Noise Strength=&MergeHelper_FloatSwitch(0=0, 1=200, 2=?, 3=?, Index=&SpawnGroup(), Num=::NewEnumerator0)
    Pan Noise=true
    Pan Noise Field=<0 0 1>
   )
   &Drag(
    Drag=&MergeHelper_FloatSwitch(
     0=&FloatFromCurve(FloatCurve=DataInterface() {propertyValues {Curve {keys [<at 0 1>,<at 1 0>]}}})
     1=&FloatFromCurve(
      FloatCurve=DataInterface() {propertyValues {Curve {
         defaultValue=0
         keys [
          {arriveTangent=-0.1861, interpMode=RCIM_Cubic, leaveTangent=-0.1861, tangentMode=RCTM_User, time=0, value=0}
          {arriveTangent=-1.5452, interpMode=RCIM_Cubic, leaveTangent=-1.5452, tangentMode=RCTM_User, time=0.4054, value=0.8656}
          {arriveTangent=-0.0405, interpMode=RCIM_Cubic, leaveTangent=-0.0405, tangentMode=RCTM_User, time=1, value=0.3}
         ]
        }}}
      Scale Curve=&RandomRangeFloat(Minimum=15, Maximum=35)
     )
     2=?
     3=?
     Index=&SpawnGroup()
     Num=::NewEnumerator0
    )
    Use Linear Drag=true
    Use Rotational Drag=false
   )
   &GravityForce(Gravity=&MergeHelper_VectorSwitch(0=<0 0 0>, 1=<0 0 0>, 2=?, 3=?, Index=&SpawnGroup(), Num=::NewEnumerator0))
   &ScaleColor(
    Scale Alpha=&MergeHelper_FloatSwitch(
     0=&FloatFromCurve(
      FloatCurve=DataInterface() {propertyValues {Curve {
         defaultValue=1.0113
         keys [
          {arriveTangent=9.0725, interpMode=RCIM_Cubic, leaveTangent=9.0725, tangentMode=RCTM_User, time=0, value=0}
          {arriveTangent=-0.6257, interpMode=RCIM_Cubic, leaveTangent=-0.6257, tangentMode=RCTM_User, time=0.1408, value=0.9847}
          {arriveTangent=-2.4217, interpMode=RCIM_Cubic, leaveTangent=-2.4217, tangentMode=RCTM_User, time=1, value=0}
         ]
        }}}
      Scale Curve=&FNTimeOfDayScalar(Night Scale=1, Morning Scale=1, Evening Scale=1)
     )
     1=&FloatFromCurve(
      FloatCurve=DataInterface() {propertyValues {Curve {
         defaultValue=1.0113
         keys [
          {arriveTangent=9.0725, interpMode=RCIM_Cubic, leaveTangent=9.0725, tangentMode=RCTM_User, time=0, value=0}
          {arriveTangent=-2.5249, interpMode=RCIM_Cubic, leaveTangent=-2.5249, tangentMode=RCTM_User, time=0.1408, value=0.9847}
          {arriveTangent=-0.2598, interpMode=RCIM_Cubic, leaveTangent=-0.2598, time=0.3357, value=0.588}
          {arriveTangent=-0.3364, interpMode=RCIM_Cubic, leaveTangent=-0.3364, time=0.4307, value=0.9093}
          {arriveTangent=-0.4029, interpMode=RCIM_Cubic, leaveTangent=-0.4029, time=0.5597, value=0.5127}
          {arriveTangent=-1.7645, interpMode=RCIM_Cubic, leaveTangent=-1.7645, time=0.6, value=0.8411}
          {arriveTangent=-2.1027, interpMode=RCIM_Cubic, leaveTangent=-2.1027, time=0.797, value=0.0939}
          {arriveTangent=-0.0405, interpMode=RCIM_Cubic, leaveTangent=-0.0405, tangentMode=RCTM_User, time=1, value=0}
         ]
        }}}
      Scale Curve=&FNTimeOfDayScalar(Night Scale=2, Morning Scale=1, Evening Scale=1)
     )
     2=?
     3=?
     Index=&SpawnGroup()
     Num=::NewEnumerator0
    )
    ScaleA=true
    ScaleRGB=true
   )
   &SolveForcesAndVelocity(Clamp Velocity=false, Limit Acceleration=false)
   /SetVariables(
    Particles.GlowCamOffset=-5
    Particles.GlowSize=&Multiply_Vector2D_ByFloat(Float=&RandomRangeFloat(Minimum=0.1, Maximum=0.4), Vector2D=$Particles.SpriteSize)
    Particles.SpriteGlowColor:LinearColor=&Multiply_LinearColorByFloat(
     Linear Color=<rgba 15 1.6873 1.8747 2>
     Float=&Multiply_Float(
      A=&FloatFromCurve(FloatCurve=DataInterface() {propertyValues {Curve {
          defaultValue=1.0113
          keys [
           {arriveTangent=9.0725, interpMode=RCIM_Cubic, leaveTangent=9.0725, tangentMode=RCTM_User, time=0, value=0}
           {arriveTangent=-2.5249, interpMode=RCIM_Cubic, leaveTangent=-2.5249, tangentMode=RCTM_User, time=0.1408, value=0.9847}
           {arriveTangent=-0.2598, interpMode=RCIM_Cubic, leaveTangent=-0.2598, time=0.427, value=0.334}
           {arriveTangent=-1.024, interpMode=RCIM_Cubic, leaveTangent=-1.024, time=0.4307, value=0.9093}
           {arriveTangent=-1.1872, interpMode=RCIM_Cubic, leaveTangent=-1.1872, time=0.5924, value=0.1646}
           {arriveTangent=-0.3456, interpMode=RCIM_Cubic, leaveTangent=-0.3456, time=0.607, value=0.7}
           {arriveTangent=-1.7812, interpMode=RCIM_Cubic, leaveTangent=-1.7812, time=0.797, value=0.0939}
           {arriveTangent=-0.0405, interpMode=RCIM_Cubic, leaveTangent=-0.0405, tangentMode=RCTM_User, time=1, value=0}
          ]
         }}})
      B=&RandomRangeFloat(Minimum=0.1, Maximum=0.4)
     )
    )
   ) {isSetParameters=true}
   /SetVariables(Particles.Kiss Position:NiagaraPosition=&AddVectorToPosition(Position=$Particles.Position, Vector=&MakeVector(X=0, Y=0, Z=0))) {isSetParameters=true}
  ]
  rendererClasses [&NiagaraSpriteRendererProperties]
  renderers [
   {rendererClass=&NiagaraSpriteRendererProperties}
   {rendererClass=&NiagaraSpriteRendererProperties, rendererIndex=1}
   {rendererClass=&NiagaraSpriteRendererProperties, rendererIndex=2}
  ]
 }}
summary {
 systemName=NS_Emote_RowLiaison_hearts
 emitters [{emitterName=Sparks_and_Hearts003, rendererClasses [&NiagaraSpriteRendererProperties]}]
 userVariables []
}
