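# Niagara system for the RowLiaison emote that emits short-lived pink-red hearts from the gadget's FX_Phone_socket.
# Two spawn groups feed one emitter with mesh and sprite renderers, tagging particles by group for visibility.
# Particles drift upward under drag, pop and shrink via a scale curve, and fade with time-of-day-scaled alpha.
# BlendSpace releases them from local to world space on per-group curves, switching to fully local when owner speed passes a threshold.
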NiagaraSystem /BRCosmetics/Animation/Game/MainPlayer/Emotes/RowLiaison/FX/NS_Emote_RowLiaison_phoneHearts
dependencies {
 usedDataInterfaces [&NiagaraDataInterfaceSkeletalMesh, &NiagaraDataInterfaceCurve]
 usedDynamicInputs []
 usedModules [&SystemState, &BlendSpaceHelper]
}
emitters {LittleHearts {
  emitterUpdateScript [
   &EmitterState(Life Cycle Mode=::NewEnumerator1, Loop Behavior=::NewEnumerator1, Loop Duration=0.25, UseLoopDelay=false)
   &SpawnRate(SpawnRate=25, Use Spawn Probability=false) {module=SpawnRate001}
   &SpawnBurst_Instantaneous(Spawn Count=15, Use Spawn Probability=false, UseLoopCountLimit=false) {enabled=false}
   &SpawnRate(SpawnGroup=1, SpawnRate=25, Use Spawn Probability=false)
  ]
  particleSpawnScript [
   /SetVariables(Particles.VisibilityTag=&SpawnGroup()) {isSetParameters=true}
   &InitializeParticle(
    Color=&RandomRangeLinearColor(Minimum=<rgba 1 0 0.0162 1>, Maximum=<rgba 1 0 0.1857 1>)
    Lifetime=&RandomRangeFloat(Minimum=0.35, Maximum=0.6)
    Mesh Scale=&VectorFromFloat(Value=&RandomRangeFloat(Minimum=0.05, Maximum=0.18))
    Mesh Scale Mode=::NewEnumerator4
    Sprite Size Mode=::NewEnumerator3
    Uniform Sprite Size=&RandomRangeFloat(Minimum=25, Maximum=50)
    UsePositionOffset=false
    Write Material Random=true
   )
   &AddVelocity(
    Rotation Coordinate Space=::Simulation
    Velocity=&RandomRangeVector(Minimum=<-35 30 50>, Maximum=<35 -30 150>)
   )
   &SkeletalMeshLocation(
    MaskSpawnGroup=false
    Mesh Sampling Type=::NewEnumerator5
    Position Sampling=::NewEnumerator1
    Sampled Position Offset=<2 0 -3>
    Skeletal Mesh=DataInterface() {propertyValues {
      DefaultMesh=null
      ExcludeBoneName=root
      OptionalComponentClass=null
      PreviewMesh=/BRCosmetics/Animation/Props/Gadgets_RowLiaison/Meshes/RowLiaison_Gadget
      SkinningMode=SkinOnTheFly
      SoftSourceActor=null
      SourceMode=Default
      UvSetIndex=0
      WholeMeshLOD=-1
      bExcludeBone=false
      bReadDeformedGeometry=true
      bRequireCurrentFrameData=true
      ComponentTags []
      FilteredBones []
      FilteredSockets [FX_Phone_socket]
      MeshUserParameter {parameter {name=None, typeDefHandle {registeredTypeIndex=1}, varData []}}
      SamplingRegions []
     }}
    Socket Sampling Mode=::NewEnumerator5
   )
   &InitialMeshOrientation(
    Mesh Orientation Mode=::NewEnumerator6
    Orientation Quaternion:Quat4f=&MakeQuatFromAxisAngle(Axis=&RandomRangeVector(Minimum=<-0.05 -0.03 0>, Maximum=<0.05 0.03 0>), AngleInDegrees=90)
    Rotation=<0 0 0.25>
    Use Rotation Vector=true
   )
  ]
  particleUpdateScript [
   &ParticleState()
   &ScaleMeshSize(Scale Factor=&VectorFromFloat(Value=&FloatFromCurve(
      FloatCurve=DataInterface() {propertyValues {Curve {keys [
          {arriveTangent=3.6183, interpMode=RCIM_Cubic, time=0, value=0.2}
          {arriveTangent=2.905, interpMode=RCIM_Cubic, leaveTangent=2.905, time=0.047, value=1.0124}
          {arriveTangent=0.8867, interpMode=RCIM_Cubic, leaveTangent=0.8867, time=0.2882, value=1.0373}
          {arriveTangent=-2.4737, interpMode=RCIM_Cubic, leaveTangent=-2.4737, time=0.4134, value=1.3373}
          {interpMode=RCIM_Cubic, leaveTangent=-33.019, time=0.7075, value=0}
         ]}}}
      Scale Curve=0.5
     )))
   &Drag(Drag=&RandomRangeFloat(Minimum=2, Maximum=8), Use Linear Drag=true, Use Rotational Drag=true)
   &AccelerationForce(Acceleration=&RandomRangeVector(Minimum=<0 0 30>, Maximum=<0 0 60>))
   &SolveForcesAndVelocity(Clamp Velocity=false, Limit Acceleration=false)
   &ScaleSpriteSizeBySpeed(Sample Scale Factor By Curve=true)
   &BlendSpace("Position Blend (World <-> Local)":NiagaraFloat=&MakeCustomFloatFromBool(
     Bool=&SetBoolByFloatComparison(A=&VectorLength(Vector=$Engine.Owner.Velocity), B=400)
     True Float=1
     False Float=&MergeHelper_FloatSwitch(
      0=&FloatFromCurve(FloatCurve=DataInterface() {propertyValues {Curve {keys [<at 0 1 RCIM_Cubic>, {arriveTangent=-3.5186, interpMode=RCIM_Cubic, leaveTangent=-3.5186, tangentMode=RCTM_User, time=0.8, value=0}]}}})
      1=&FloatFromCurve(FloatCurve=DataInterface() {propertyValues {Curve {keys [<at 0 1 RCIM_Cubic>, {arriveTangent=-3.5186, interpMode=RCIM_Cubic, leaveTangent=-3.5186, tangentMode=RCTM_User, time=0.5, value=0}]}}})
      2=?
      3=?
      Index=&SpawnGroup()
      Num=::NewEnumerator0
     )
     Translation Contribution:NiagaraFloat=1
     Rotation Contribution:NiagaraFloat=1
    )) {module=BlendSpace002}
   &ScaleColor(
    Scale Alpha=&FloatFromCurve(
     FloatCurve=DataInterface() {propertyValues {Curve {keys [<at 0 1>,<at 1 0>]}}}
     Scale Curve=&FNTimeOfDayScalar(Night Scale=5, Morning Scale=1, Day Scale=5, Evening Scale=1)
    )
    ScaleA=true
    ScaleRGB=true
   )
   &DynamicMaterialParameters(
    Index 0 Param 2=1
    Module.Index0Float0Write=true
    Module.Index0Float1Write=true
    Module.Index0Float2Write=true
    Module.Index0Float3Write=true
   )
  ]
  rendererClasses [&NiagaraMeshRendererProperties, &NiagaraSpriteRendererProperties]
  renderers [
   {rendererClass=&NiagaraMeshRendererProperties, rendererIndex=0}
   {rendererClass=&NiagaraSpriteRendererProperties, rendererIndex=1}
  ]
 }}
summary {
 systemName=NS_Emote_RowLiaison_phoneHearts
 emitters [{emitterName=LittleHearts, rendererClasses [&NiagaraMeshRendererProperties, &NiagaraSpriteRendererProperties]}]
 userVariables [
  User.bIsFrontEnd {type=NiagaraBool}
  User.bIsFrontEndPreview {type=NiagaraBool}
  User.NormalizedNotifyProgress {type=NiagaraFloat, defaultValue=0, description=""}
 ]
}
